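# Control Rig for the LemurClam large male body, driving procedural deformation of its robot limbs, armor, skiblades and levers.
# After a shared DeformRig function, it runs leg, arm and lever passes using constraints and weighted transform offsets.
# A DeltaRotation helper remaps swing/twist against construction-time nulls into corrective weights.
# Lever sticks follow smoothed DeformRig attributes and movement inputs; FE idle and Kalman-smoothed targeting variables adjust arms and gear.
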ControlRigBlueprint /BRCosmetics/Characters/Player/Male/Large/Bodies/M_LRG_LemurClam/Meshes/M_LRG_LemurClam_CtrlRig
graphs [
 New Function_ContainedGraph {nodes [
   n1 = RigVMFunctionEntryNode()
   RigVMFunction_MathBoolToInteger = &RigVMFunction_MathBoolToInteger(Value=@n1.onlyTwist)
   RigVMFunction_MathQuaternionSwingTwist = &RigVMFunction_MathQuaternionSwingTwist(
    twistAxis=@n1.twistAxis
    Input=&RigUnit_GetRelativeTransformForItem(bChildInitial=False, bParentInitial=False, child=@n1.child, parent=@n1.parent).RelativeTransform.Rotation
   )
   n2 = RigVMDispatchNode(
    Values=((),())
    Index=@RigVMFunction_MathBoolToInteger.Result
    Values.0=@RigVMFunction_MathQuaternionSwingTwist.Swing
    Values.1=@RigVMFunction_MathQuaternionSwingTwist.Twist
   ) {template="DISPATCH_RigVMDispatch_SelectInt32(in Index,in Values,out Result)"}
   RigVMFunction_MathQuaternionToEuler = &RigVMFunction_MathQuaternionToEuler(Value=@n2.Result, RotationOrder=@n1.Rotation Order)
   > DISPATCH_RigVMDispatch_Print = RigVMDispatchNode(ScreenColor=(R=1,G=1,B=1,A=1), ScreenDuration=1.5, Enabled=@n1.debugRotation, Value=@RigVMFunction_MathQuaternionToEuler.Result) {template="DISPATCH_RigVMDispatch_Print(in Prefix,in Value,in Enabled,in ScreenDuration,in ScreenColor)"}
   Cast = RigVMDispatchNode(Value=@n1.triggerAxis) {template="DISPATCH_RigVMDispatch_CastEnumToInt(in Value,out Result)"}
   DISPATCH_RigVMDispatch_SelectInt32 = RigVMDispatchNode(
    Values=(0,0,0)
    Index=@Cast.Result
    Values.0=@RigVMFunction_MathQuaternionToEuler.Result.X
    Values.1=@RigVMFunction_MathQuaternionToEuler.Result.Y
    Values.2=@RigVMFunction_MathQuaternionToEuler.Result.Z
   ) {template="DISPATCH_RigVMDispatch_SelectInt32(in Index,in Values,out Result)"}
   Remap = &RigVMFunction_MathDoubleRemap(
    bClamp=False
    Value=@DISPATCH_RigVMDispatch_SelectInt32.Result
    TargetMinimum=@n1.outputMinMax.X
    TargetMaximum=@n1.outputMinMax.Y
    SourceMinimum=@n1.rotationMinMax.X
    SourceMaximum=@n1.rotationMinMax.Y
   )
   Absolute = &RigVMFunction_MathDoubleAbs(Value=@Remap.Result)
   Multiply = &RigVMFunction_MathDoubleMul(
    A=&RigVMFunction_MathDoubleClamp(Value=RigVMDispatchNode(True=@Absolute.Result, Condition=@n1.absoluteValue, False=@Remap.Result) {template="DISPATCH_RigVMDispatch_If(in Condition,in True,in False,out Result)"}.Result).Result
    B=@n1.premult
   )
   > RigVMFunctionReturnNode(outputQuat=@n2.Result, outputWeight=@Multiply.Result)
  ]}
 New Function_ContainedGraph {nodes [
   n1 = RigVMFunctionEntryNode()
   RigVMFunction_ControlFlowBranch = &RigVMFunction_ControlFlowBranch(ExecuteContext=@n1.ExecuteContext, Condition=RigVMVariableNode() {variable=FE idle}.Value)
   DeltaRotation v4 = RigVMFunctionReferenceNode(
    Rotation Order=ZYX
    child=(Type=Bone,Name="hand_l")
    outputMinMax=(X=0,Y=1)
    parent=(Type=Null,Name="hand_l_null")
    premult=1
    rotationMinMax=(X=10,Y=120)
    triggerAxis=Z
    twistAxis=(X=1,Y=0,Z=0)
    ExecuteContext=@RigVMFunction_ControlFlowBranch.Completed
   )
   > &RigUnit_ModifyTransforms(
    ItemToModify="((Item=(Type=Bone,Name=\"skiblades\"),Transform=(Rotation=(X=0,Y=-0.6088,Z=0,W=0.7934),Translation=(X=0,Y=0,Z=0),Scale3D=(X=1,Y=1,Z=1))),(Item=(Type=Bone,Name=\"skiblade_rope_A\"),Transform=(Rotation=(X=-0.1739,Y=0.2219,Z=0.0402,W=-0.9586),Translation=(X=-11.9 ...(truncated)"
    WeightMinimum=0
    Weight=@"DeltaRotation v4".outputWeight
   )
   > DeltaRotation v4_6 = RigVMFunctionReferenceNode(
    Rotation Order=ZYX
    absoluteValue=false
    child=(Type=Bone,Name="hand_l")
    outputMinMax=(X=0,Y=1)
    parent=(Type=Null,Name="hand_l_null")
    premult=1
    rotationMinMax=(X=-10,Y=-60)
    triggerAxis=Y
    twistAxis=(X=1,Y=0,Z=0)
   )
   > DeltaRotation v4_7 = RigVMFunctionReferenceNode(
    Rotation Order=ZYX
    child=(Type=Bone,Name="hand_l")
    onlyTwist=true
    outputMinMax=(X=0,Y=1)
    parent=(Type=Null,Name="hand_l_null")
    premult=1
    rotationMinMax=(X=0,Y=120)
    triggerAxis=X
    twistAxis=(X=1,Y=0,Z=0)
   )
   > &RigUnit_ModifyTransforms(
    ItemToModify="((Item=(Type=Bone,Name=\"skiblades\"),Transform=(Rotation=(X=0,Y=-0.5,Z=0,W=0.866),Translation=(X=0,Y=0,Z=0),Scale3D=(X=1,Y=1,Z=1))),(Item=(Type=Bone,Name=\"skiblade_rope_A\"),Transform=(Rotation=(X=0,Y=0,Z=0,W=1),Translation=(X=-6.1094,Y=0,Z=-14.4311),Scale3D ...(truncated)"
    WeightMinimum=0
    Weight=&RigVMFunction_MathDoubleMul(A=@"DeltaRotation v4_6".outputWeight, B=@"DeltaRotation v4_7".outputWeight).Result
   )
   > DeltaRotation v4_1 = RigVMFunctionReferenceNode(
    Rotation Order=ZYX
    child=(Type=Bone,Name="hand_r")
    outputMinMax=(X=0,Y=1)
    parent=(Type=Null,Name="hand_r_null")
    premult=1
    rotationMinMax=(X=45,Y=90)
    triggerAxis=Z
    twistAxis=(X=1,Y=0,Z=0)
   )
   > &RigUnit_ModifyTransforms(
    ItemToModify=((Item=(Type=Bone,Name="MSTR_shield_r"),Transform=(Rotation=(X=0,Y=0,Z=0.0421,W=-0.9991),Translation=(X=2.5104,Y=0,Z=0),Scale3D=(X=1,Y=1,Z=1))))
    WeightMinimum=0
    Weight=@"DeltaRotation v4_1".outputWeight
   )
   > &RigUnit_ParentConstraint(
    AdvancedSettings=(InterpolationType=Average,RotationOrderForFilter=XZY)
    Child=(Type=Bone,Name="robot_arm_transition_l")
    Parents=((Item=(Type=Bone,Name="spine_05"),Weight=0.5),(Item=(Type=Bone,Name="robot_arm_l"),Weight=0.5))
    Weight=1
   )
   > DeltaRotation v4_2 = RigVMFunctionReferenceNode(
    Rotation Order=YXZ
    child=(Type=Bone,Name="robot_arm_l")
    outputMinMax=(X=0,Y=1)
    parent=(Type=Null,Name="upperarm_l_null")
    premult=1
    rotationMinMax=(X=0,Y=-70)
    triggerAxis=Y
    twistAxis=(X=1,Y=0,Z=0)
   )
   > &RigUnit_ModifyTransforms(
    ItemToModify="((Item=(Type=Bone,Name=\"shoulderhelmet_l\"),Transform=(Rotation=(X=0,Y=0.1518,Z=0,W=0.9884),Translation=(X=9,Y=0,Z=-1),Scale3D=(X=1,Y=1,Z=1))),(Item=(Type=Bone,Name=\"robot_arm_transition_l\"),Transform=(Rotation=(X=0,Y=0,Z=0,W=-1),Translation=(X=0,Y=-5,Z=10) ...(truncated)"
    WeightMinimum=0
    Weight=@"DeltaRotation v4_2".outputWeight
   )
   > DeltaRotation v4_8 = RigVMFunctionReferenceNode(
    Rotation Order=YXZ
    child=(Type=Bone,Name="robot_arm_l")
    outputMinMax=(X=0,Y=1)
    parent=(Type=Null,Name="upperarm_l_null")
    premult=1
    rotationMinMax=(X=0,Y=-120)
    triggerAxis=Y
    twistAxis=(X=1,Y=0,Z=0)
   )
   > &RigUnit_ModifyTransforms(
    ItemToModify=((Item=(Type=Bone,Name="robot_arm_transition_l"),Transform=(Rotation=(X=-0.5083,Y=0.1383,Z=0.2282,W=-0.8188),Translation=(X=7.6764,Y=8.577,Z=-1.1515),Scale3D=(X=0.6321,Y=2.1043,Z=1.1292))))
    WeightMinimum=0
    Weight=@"DeltaRotation v4_8".outputWeight
   )
   Subtract = &RigVMFunction_MathDoubleSub(A=1, B=RigVMVariableNode() {variable=FE idle alpha}.Value)
   > &RigUnit_ModifyTransforms(
    ItemToModify=((Item=(Type=Bone,Name="robot_arm_l"),Transform=(Rotation=(X=0,Y=0,Z=0,W=-1),Translation=(X=3,Y=0,Z=-5),Scale3D=(X=1,Y=1,Z=1))))
    WeightMinimum=0
    Weight=&RigVMFunction_MathDoubleMul(A=@"DeltaRotation v4_8".outputWeight, B=@Subtract.Result).Result
   )
   > DeltaRotation v4_3 = RigVMFunctionReferenceNode(
    Rotation Order=YXZ
    child=(Type=Bone,Name="robot_arm_l")
    outputMinMax=(X=0,Y=1)
    parent=(Type=Null,Name="upperarm_l_null")
    premult=1
    rotationMinMax=(X=0,Y=70)
    triggerAxis=Z
    twistAxis=(X=1,Y=0,Z=0)
   )
   > &RigUnit_ModifyTransforms(
    ItemToModify=((Item=(Type=Bone,Name="robot_arm_transition_l"),Transform=(Rotation=(X=0,Y=0,Z=-0.2748,W=0.9615),Translation=(X=-4.1219,Y=-4.9784,Z=-3.143),Scale3D=(X=1,Y=1.5,Z=1))))
    WeightMinimum=0
    Weight=@"DeltaRotation v4_3".outputWeight
   )
   Multiply = &RigVMFunction_MathDoubleMul(A=@"DeltaRotation v4_3".outputWeight, B=@Subtract.Result)
   > &RigUnit_ModifyTransforms(
    ItemToModify=((Item=(Type=Bone,Name="robot_arm_l"),Transform=(Rotation=(X=0,Y=0,Z=0.0872,W=0.9962),Translation=(X=0,Y=-8,Z=0),Scale3D=(X=1,Y=1,Z=1))))
    WeightMinimum=0
    Weight=@Multiply.Result
   )
   RigVMFunction_AnimEasing = /Script/RigVM.RigVMFunction_AnimEasing(
    SourceMaximum=1
    SourceMinimum=0
    TargetMaximum=1
    TargetMinimum=0
    Type=QuinticEaseOut
    Value=&RigVMFunction_MathDoubleMul(A=@"DeltaRotation v4_2".outputWeight, B=@"DeltaRotation v4_3".outputWeight).Result
   ) {is=struct RigVMUnitNode}
   > &RigUnit_ModifyTransforms(
    ItemToModify=((Item=(Type=Bone,Name="robot_arm_transition_l"),Transform=(Rotation=(X=0,Y=0,Z=0.0872,W=0.9962),Translation=(X=-5,Y=-2,Z=0),Scale3D=(X=1,Y=1,Z=1))))
    WeightMinimum=0
    Weight=@RigVMFunction_AnimEasing.Result
   )
   > &RigUnit_ModifyTransforms(
    ItemToModify=((Item=(Type=Bone,Name="robot_arm_l"),Transform=(Rotation=(X=0,Y=0,Z=0,W=1),Translation=(X=-4,Y=0,Z=5),Scale3D=(X=1,Y=1,Z=1))))
    WeightMinimum=0
    Weight=&RigVMFunction_MathDoubleMul(A=@RigVMFunction_AnimEasing.Result, B=@Subtract.Result).Result
   )
   > &RigUnit_ParentConstraint(
    AdvancedSettings=(InterpolationType=Average,RotationOrderForFilter=XZY)
    Child=(Type=Bone,Name="shoulderhelmet_A_l")
    Parents=((Item=(Type=Bone,Name="robot_arm_l"),Weight=0.75),(Item=(Type=Bone,Name="shoulderhelmet_l"),Weight=0.25))
    Weight=1
   )
   > &RigUnit_ParentConstraint(
    AdvancedSettings=(InterpolationType=Average,RotationOrderForFilter=XZY)
    Child=(Type=Bone,Name="shoulderhelmet_B_l")
    Parents=((Item=(Type=Bone,Name="robot_arm_l"),Weight=0.75),(Item=(Type=Bone,Name="shoulderhelmet_l"),Weight=0.25))
    Weight=1
   )
   > &RigUnit_ParentConstraint(
    AdvancedSettings=(InterpolationType=Average,RotationOrderForFilter=XZY)
    Child=(Type=Bone,Name="robot_arm_transition_r")
    Parents=((Item=(Type=Bone,Name="spine_05"),Weight=0.5),(Item=(Type=Bone,Name="robot_arm_r"),Weight=0.5))
    Weight=1
   )
   > DeltaRotation v4_4 = RigVMFunctionReferenceNode(
    Rotation Order=YXZ
    child=(Type=Bone,Name="robot_arm_r")
    outputMinMax=(X=0,Y=1)
    parent=(Type=Null,Name="upperarm_r_null")
    premult=1
    rotationMinMax=(X=0,Y=-70)
    triggerAxis=Y
    twistAxis=(X=1,Y=0,Z=0)
   )
   > &RigUnit_ModifyTransforms(
    ItemToModify="((Item=(Type=Bone,Name=\"shoulderhelmet_r\"),Transform=(Rotation=(X=0,Y=-0.1518,Z=0,W=-0.9884),Translation=(X=-9,Y=0,Z=1),Scale3D=(X=1,Y=1,Z=1))),(Item=(Type=Bone,Name=\"robot_arm_transition_r\"),Transform=(Rotation=(X=0,Y=0,Z=0,W=1),Translation=(X=0,Y=5,Z=-10 ...(truncated)"
    WeightMinimum=0
    Weight=@"DeltaRotation v4_4".outputWeight
   )
   > DeltaRotation v4_9 = RigVMFunctionReferenceNode(
    Rotation Order=YXZ
    child=(Type=Bone,Name="robot_arm_r")
    outputMinMax=(X=0,Y=1)
    parent=(Type=Null,Name="upperarm_r_null")
    premult=1
    rotationMinMax=(X=0,Y=-120)
    triggerAxis=Y
    twistAxis=(X=1,Y=0,Z=0)
   )
   > &RigUnit_ModifyTransforms(
    ItemToModify=((Item=(Type=Bone,Name="robot_arm_transition_r"),Transform=(Rotation=(X=0.5083,Y=-0.1383,Z=-0.2282,W=0.8188),Translation=(X=-7.6764,Y=-8.577,Z=1.1515),Scale3D=(X=0.6321,Y=2.1043,Z=1.1292))))
    WeightMinimum=0
    Weight=@"DeltaRotation v4_9".outputWeight
   )
   n2 = &RigVMFunction_MathDoubleSub(A=1, B=RigVMVariableNode() {variable=FE idle alpha}.Value)
   > &RigUnit_ModifyTransforms(
    ItemToModify=((Item=(Type=Bone,Name="robot_arm_r"),Transform=(Rotation=(X=0,Y=0,Z=0,W=-1),Translation=(X=-3,Y=0,Z=5),Scale3D=(X=1,Y=1,Z=1))))
    WeightMinimum=0
    Weight=&RigVMFunction_MathDoubleMul(A=@"DeltaRotation v4_9".outputWeight, B=@n2.Result).Result
   )
   > DeltaRotation v4_5 = RigVMFunctionReferenceNode(
    Rotation Order=YXZ
    child=(Type=Bone,Name="robot_arm_r")
    outputMinMax=(X=0,Y=1)
    parent=(Type=Null,Name="upperarm_r_null")
    premult=1
    rotationMinMax=(X=0,Y=70)
    triggerAxis=Z
    twistAxis=(X=1,Y=0,Z=0)
   )
   > &RigUnit_ModifyTransforms(
    ItemToModify=((Item=(Type=Bone,Name="robot_arm_transition_r"),Transform=(Rotation=(X=0,Y=0,Z=0,W=1),Translation=(X=2.5,Y=6,Z=0.6613),Scale3D=(X=1,Y=1.5,Z=1))))
    WeightMinimum=0
    Weight=@"DeltaRotation v4_5".outputWeight
   )
   > &RigUnit_ModifyTransforms(
    ItemToModify=((Item=(Type=Bone,Name="robot_arm_r"),Transform=(Rotation=(X=0,Y=0,Z=0.0872,W=0.9962),Translation=(X=0,Y=8,Z=0),Scale3D=(X=1,Y=1,Z=1))))
    WeightMinimum=0
    Weight=&RigVMFunction_MathDoubleMul(A=@"DeltaRotation v4_5".outputWeight, B=@n2.Result).Result
   )
   n3 = &RigVMFunction_AnimEasing(
    SourceMaximum=1
    SourceMinimum=0
    TargetMaximum=1
    TargetMinimum=0
    Type=QuinticEaseOut
    Value=&RigVMFunction_MathDoubleMul(A=@"DeltaRotation v4_4".outputWeight, B=@"DeltaRotation v4_5".outputWeight).Result
   )
   > &RigUnit_ModifyTransforms(
    ItemToModify=((Item=(Type=Bone,Name="robot_arm_transition_r"),Transform=(Rotation=(X=0,Y=0,Z=0.0872,W=0.9962),Translation=(X=5,Y=2,Z=0),Scale3D=(X=1,Y=1,Z=1))))
    WeightMinimum=0
    Weight=@n3.Result
   )
   > &RigUnit_ModifyTransforms(
    ItemToModify=((Item=(Type=Bone,Name="robot_arm_r"),Transform=(Rotation=(X=0,Y=0,Z=0,W=1),Translation=(X=4,Y=0,Z=-5),Scale3D=(X=1,Y=1,Z=1))))
    WeightMinimum=0
    Weight=&RigVMFunction_MathDoubleMul(B=@n3.Result, A=@n2.Result).Result
   )
   > &RigUnit_ParentConstraint(
    AdvancedSettings=(InterpolationType=Average,RotationOrderForFilter=XZY)
    Child=(Type=Bone,Name="shoulderhelmet_A_r")
    Parents=((Item=(Type=Bone,Name="robot_arm_r"),Weight=0.75),(Item=(Type=Bone,Name="shoulderhelmet_r"),Weight=0.25))
    Weight=1
   )
   > &RigUnit_ParentConstraint(
    AdvancedSettings=(InterpolationType=Average,RotationOrderForFilter=XZY)
    Child=(Type=Bone,Name="shoulderhelmet_B_r")
    Parents=((Item=(Type=Bone,Name="robot_arm_r"),Weight=0.75),(Item=(Type=Bone,Name="shoulderhelmet_r"),Weight=0.25))
    Weight=1
   )
   Cast = &RigVMFunction_MathBoolToFloat(Value=RigVMVariableNode() {variable=is_targeting}.Value)
   n4 = &RigVMFunction_AnimEasing(
    SourceMaximum=1
    SourceMinimum=0
    TargetMaximum=1
    TargetMinimum=0
    Type=CubicEaseOut
    Value=&RigVMFunction_KalmanFloat(BufferSize=13, Value=@Cast.Result) {template="KalmanFilter::Execute(in Value,in BufferSize,out Result)"}.Result
   )
   > &RigUnit_ModifyTransforms(
    ItemToModify="((Item=(Type=Bone,Name=\"MSTR_shoulderhelmet_r\"),Transform=(Rotation=(X=0.342,Y=0,Z=0,W=0.9397),Translation=(X=0,Y=5,Z=2),Scale3D=(X=1,Y=1,Z=1))),(Item=(Type=Bone,Name=\"MSTR_shield_r\"),Transform=(Rotation=(X=0.3007,Y=0,Z=0,W=0.9537),Translation=(X=0,Y=5,Z=0 ...(truncated)"
    WeightMinimum=0
    Weight=@n4.Result
   )
   > &RigUnit_ModifyTransforms(
    ItemToModify="((Item=(Type=Bone,Name=\"dyn_skiblade_A\"),Transform=(Rotation=(X=0.3552,Y=0.0619,Z=-0.2057,W=0.9097),Translation=(X=5,Y=-5,Z=0),Scale3D=(X=1,Y=1,Z=1))),(Item=(Type=Bone,Name=\"dyn_skiblade_B\"),Transform=(Rotation=(X=0.2394,Y=0.1299,Z=-0.204,W=0.9403),Transla ...(truncated)"
    WeightMinimum=0
    Weight=@n4.Result
   )
   > RigVMFunctionReturnNode()
   &RigUnit_ParentConstraint(
    AdvancedSettings=(InterpolationType=Average,RotationOrderForFilter=XZY)
    Child=(Type=Bone,Name="robot_arm_l")
    Filter=(TranslationFilter=(bX=true,bY=true,bZ=true),RotationFilter=(bX=false,bY=false,bZ=false),ScaleFilter=(bX=false,bY=false,bZ=false))
    Parents=((Item=(Type=Bone,Name="spine_05"),Weight=0.65),(Item=(Type=Bone,Name="clavicle_l"),Weight=0.35))
    Weight=1
    ExecutePin=@RigVMFunction_ControlFlowBranch.False
   )
   > &RigUnit_ParentConstraint(
    AdvancedSettings=(InterpolationType=Average,RotationOrderForFilter=XZY)
    Child=(Type=Bone,Name="robot_arm_r")
    Filter=(TranslationFilter=(bX=true,bY=true,bZ=true),RotationFilter=(bX=false,bY=false,bZ=false),ScaleFilter=(bX=false,bY=false,bZ=false))
    Parents=((Item=(Type=Bone,Name="spine_05"),Weight=0.65),(Item=(Type=Bone,Name="clavicle_r"),Weight=0.35))
    Weight=1
   )
   > &RigUnit_ParentConstraint(
    AdvancedSettings=(InterpolationType=Average,RotationOrderForFilter=XZY)
    Child=(Type=Null,Name="upperarm_l_null")
    Filter=(TranslationFilter=(bX=true,bY=true,bZ=true),RotationFilter=(bX=true,bY=true,bZ=true),ScaleFilter=(bX=false,bY=false,bZ=false))
    Parents=((Item=(Type=Bone,Name="spine_05"),Weight=1))
    Weight=1
   )
   > &RigUnit_ParentConstraint(
    AdvancedSettings=(InterpolationType=Average,RotationOrderForFilter=XZY)
    Child=(Type=Null,Name="upperarm_r_null")
    Filter=(TranslationFilter=(bX=true,bY=true,bZ=true),RotationFilter=(bX=true,bY=true,bZ=true),ScaleFilter=(bX=false,bY=false,bZ=false))
    Parents=((Item=(Type=Bone,Name="spine_05"),Weight=1))
    Weight=1
   )
   > &RigUnit_AimConstraintLocalSpaceOffset(
    AdvancedSettings=(DebugSettings=(bEnabled=false,Scale=100,WorldOffset=(Rotation=(X=0,Y=0,Z=0,W=1),Translation=(X=0,Y=0,Z=0),Scale3D=(X=1,Y=1,Z=1))),RotationOrderForFilter=XZY)
    Child=(Type=Bone,Name="robot_arm_l")
    Filter=(bX=true,bY=true,bZ=true)
    Parents=((Item=(Type=Bone,Name="lowerarm_l"),Weight=1))
    Weight=1
    WorldUp=(Target=(X=0,Y=0,Z=1),Kind=Direction,Space=(Type=Bone,Name="upperarm_l"))
    bMaintainOffset=true
   )
   > &RigUnit_AimConstraintLocalSpaceOffset(
    AdvancedSettings=(DebugSettings=(bEnabled=false,Scale=100,WorldOffset=(Rotation=(X=0,Y=0,Z=0,W=1),Translation=(X=0,Y=0,Z=0),Scale3D=(X=1,Y=1,Z=1))),RotationOrderForFilter=XZY)
    AimAxis=(X=-1,Y=0,Z=0)
    Child=(Type=Bone,Name="robot_arm_r")
    Filter=(bX=true,bY=true,bZ=true)
    Parents=((Item=(Type=Bone,Name="lowerarm_r"),Weight=1))
    Weight=1
    WorldUp=(Target=(X=0,Y=0,Z=-1),Kind=Direction,Space=(Type=Bone,Name="upperarm_r"))
   )
   > /Script/ControlRig.RigUnit_PositionConstraintLocalSpaceOffset(
    Child=(Type=Bone,Name="robot_lowerarm_l")
    Filter=(bX=True,bY=True,bZ=True)
    Parents=((Item=(Type=Bone,Name="lowerarm_l"),Weight=1))
    Weight=1
    bMaintainOffset=True
   ) {is=struct RigVMUnitNode}
   > &RigUnit_PositionConstraintLocalSpaceOffset(
    Child=(Type=Bone,Name="robot_lowerarm_r")
    Filter=(bX=True,bY=True,bZ=True)
    Parents=((Item=(Type=Bone,Name="lowerarm_r"),Weight=1))
    Weight=1
    bMaintainOffset=True
   )
  ]}
 New Function_ContainedGraph {nodes [
   RigVMFunctionEntryNode()
   > DeltaRotation v4_2 = RigVMFunctionReferenceNode(
    Rotation Order=ZYX
    child=(Type=Bone,Name="calf_l")
    outputMinMax=(X=0,Y=1)
    parent=(Type=Null,Name="calf_l_null")
    premult=1
    rotationMinMax=(X=63,Y=110)
    triggerAxis=Z
    twistAxis=(X=1,Y=0,Z=0)
   )
   > &RigUnit_ModifyTransforms(
    ItemToModify="((Item=(Type=Bone,Name=\"robot_calf_l\"),Transform=(Rotation=(X=0,Y=0,Z=0,W=1),Translation=(X=-1.9132,Y=4.0358,Z=0),Scale3D=(X=1,Y=1,Z=1))),(Item=(Type=Bone,Name=\"calf_twist_02_l\"),Transform=(Rotation=(X=0,Y=0,Z=0.0277,W=0.9996),Translation=(X=-0.0178,Y=2.13 ...(truncated)"
    WeightMinimum=0
    Weight=@"DeltaRotation v4_2".outputWeight
   )
   > DeltaRotation v4_3 = RigVMFunctionReferenceNode(
    Rotation Order=ZYX
    child=(Type=Bone,Name="calf_r")
    outputMinMax=(X=0,Y=1)
    parent=(Type=Null,Name="calf_r_null")
    premult=1
    rotationMinMax=(X=63,Y=110)
    triggerAxis=Z
    twistAxis=(X=1,Y=0,Z=0)
   )
   > &RigUnit_ModifyTransforms(
    ItemToModify="((Item=(Type=Bone,Name=\"robot_calf_r\"),Transform=(Rotation=(X=0,Y=0,Z=0,W=1),Translation=(X=1.9132,Y=-4.0358,Z=0),Scale3D=(X=1,Y=1,Z=1))),(Item=(Type=Bone,Name=\"calf_twist_02_r\"),Transform=(Rotation=(X=0,Y=0,Z=0.0277,W=0.9996),Translation=(X=0.0178,Y=-2.13 ...(truncated)"
    WeightMinimum=0
    Weight=@"DeltaRotation v4_3".outputWeight
   )
   > &RigUnit_AimConstraintLocalSpaceOffset(
    AimAxis=(X=-1,Y=0,Z=0)
    Child=(Type=Bone,Name="robot_calf_l")
    Filter=(bX=false,bY=true,bZ=True)
    Parents=((Item=(Type=Bone,Name="foot_l"),Weight=1))
    UpAxis=(X=0,Y=1,Z=0)
    Weight=1
    WorldUp=(Target=(X=0,Y=0,Z=-1),Kind=Direction,Space=(Type=Bone,Name="thigh_l"))
   )
   > &RigUnit_AimConstraintLocalSpaceOffset(
    Child=(Type=Bone,Name="robot_calf_r")
    Filter=(bX=false,bY=True,bZ=True)
    Parents=((Item=(Type=Bone,Name="foot_r"),Weight=1))
    UpAxis=(X=0,Y=1,Z=0)
    Weight=1
    WorldUp=(Target=(X=0,Y=0,Z=1),Kind=Direction,Space=(Type=Bone,Name="thigh_r"))
    bMaintainOffset=true
   )
   > &RigUnit_ParentConstraint(
    AdvancedSettings=(InterpolationType=Average,RotationOrderForFilter=XZY)
    Child=(Type=Bone,Name="robot_thigh_l")
    Parents=((Item=(Type=Bone,Name="pelvis"),Weight=1))
    Weight=1
   )
   > &RigUnit_ParentConstraint(
    AdvancedSettings=(InterpolationType=Average,RotationOrderForFilter=XZY)
    Child=(Type=Bone,Name="robot_thigh_r")
    Parents=((Item=(Type=Bone,Name="pelvis"),Weight=1))
    Weight=1
   )
   > &RigUnit_AimConstraintLocalSpaceOffset(
    AimAxis=(X=-1,Y=0,Z=0)
    Child=(Type=Bone,Name="robot_thigh_l")
    Filter=(bX=True,bY=True,bZ=True)
    Parents=((Item=(Type=Bone,Name="robot_thigh_aim_l"),Weight=0.5),(Item=(Type=Bone,Name="calf_l"),Weight=0.5))
    Weight=1
    WorldUp=(Target=(X=0,Y=0,Z=-1),Kind=Direction,Space=(Type=Bone,Name="pelvis"))
   )
   > &RigUnit_AimConstraintLocalSpaceOffset(
    Child=(Type=Bone,Name="robot_thigh_r")
    Filter=(bX=True,bY=True,bZ=True)
    Parents=((Item=(Type=Bone,Name="calf_r"),Weight=0.5),(Item=(Type=Bone,Name="robot_thigh_aim_r"),Weight=0.5))
    Weight=1
    WorldUp=(Target=(X=0,Y=0,Z=1),Kind=Direction,Space=(Type=Bone,Name="pelvis"))
    bMaintainOffset=true
   )
   > DeltaRotation v4 = RigVMFunctionReferenceNode(
    Rotation Order=ZYX
    child=(Type=Bone,Name="robot_thigh_l")
    outputMinMax=(X=0,Y=1)
    parent=(Type=Null,Name="thigh_l_null")
    premult=1
    rotationMinMax=(X=-87,Y=-120)
    triggerAxis=Z
    twistAxis=(X=1,Y=0,Z=0)
   )
   > DeltaRotation v4_1 = RigVMFunctionReferenceNode(
    Rotation Order=ZYX
    child=(Type=Bone,Name="robot_thigh_r")
    outputMinMax=(X=0,Y=1)
    parent=(Type=Null,Name="thigh_r_null")
    premult=1
    rotationMinMax=(X=-87,Y=-120)
    triggerAxis=Z
    twistAxis=(X=1,Y=0,Z=0)
   )
   Maximum = &RigVMFunction_MathDoubleMax(A=@"DeltaRotation v4".outputWeight, B=@"DeltaRotation v4_1".outputWeight)
   > &RigUnit_ModifyTransforms(
    ItemToModify=((Item=(Type=Bone,Name="MSTR_handle"),Transform=(Rotation=(X=0,Y=0,Z=0.4226,W=0.9063),Translation=(X=0,Y=0,Z=0),Scale3D=(X=1,Y=1,Z=1))))
    WeightMinimum=0
    Weight=@Maximum.Result
   )
   > RigVMFunctionReturnNode()
  ]}
 New Function_ContainedGraph {nodes [
   RigVMFunctionEntryNode()
   n1 = RigVMDispatchNode(BoneName=root, Default=0, Name=deformRig_l_upperarm_up_alpha) {template="DISPATCH_RigDispatch_GetAnimAttribute(in Name,in BoneName,in Default,out Value,out Found)"}
   > /Script/ControlRig.RigUnit_SetCurveValue(Curve=l_arm_up, Value=@n1.Value) {is=struct RigVMUnitNode}
   n2 = RigVMDispatchNode(BoneName=root, Default=0, Name=deformRig_l_upperarm_fr_alpha) {template="DISPATCH_RigDispatch_GetAnimAttribute(in Name,in BoneName,in Default,out Value,out Found)"}
   > &RigUnit_SetCurveValue(Curve=l_arm_ft, Value=@n2.Value)
   n3 = RigVMVariableNode() {variable=Character Forward}
   Maximum = &RigVMFunction_MathDoubleMax(B=@n2.Value, A=@n3.Value)
   AlphaInterp = &RigVMFunction_AlphaInterp(
    InterpSpeedDecreasing=10
    InterpSpeedIncreasing=20
    Scale=1
    bClampResult=False
    Value=&RigVMFunction_MathDoubleMax(B=@n1.Value, A=@Maximum.Result).Result
   )
   > &RigUnit_ModifyTransforms(
    ItemToModify=((Item=(Type=Bone,Name="stick_l"),Transform=(Rotation=(X=0,Y=0,Z=0.2588,W=0.9659),Translation=(X=0,Y=0,Z=0),Scale3D=(X=1,Y=1,Z=1))))
    WeightMinimum=0
    Weight=@AlphaInterp.Result
   )
   n4 = RigVMDispatchNode(BoneName=root, Default=0, Name=deformRig_r_upperarm_up_alpha) {template="DISPATCH_RigDispatch_GetAnimAttribute(in Name,in BoneName,in Default,out Value,out Found)"}
   > &RigUnit_SetCurveValue(Curve=r_arm_up, Value=@n4.Value)
   n5 = RigVMDispatchNode(BoneName=root, Default=0, Name=deformRig_r_upperarm_fr_alpha) {template="DISPATCH_RigDispatch_GetAnimAttribute(in Name,in BoneName,in Default,out Value,out Found)"}
   > &RigUnit_SetCurveValue(Curve=r_arm_ft, Value=@n5.Value)
   n6 = &RigVMFunction_MathDoubleMax(B=@n5.Value, A=@n3.Value)
   n7 = &RigVMFunction_AlphaInterp(
    InterpSpeedDecreasing=10
    InterpSpeedIncreasing=20
    Scale=1
    bClampResult=False
    Value=&RigVMFunction_MathDoubleMax(A=&RigVMFunction_MathDoubleMax(B=@n4.Value, A=@n6.Result).Result, B=RigVMVariableNode() {variable=Character Move Right}.Value).Result
   )
   > &RigUnit_ModifyTransforms(
    ItemToModify=((Item=(Type=Bone,Name="stick_r"),Transform=(Rotation=(X=0,Y=0,Z=0.2588,W=0.9659),Translation=(X=0,Y=0,Z=0),Scale3D=(X=1,Y=1,Z=1))))
    WeightMinimum=0
    Weight=@n7.Result
   )
   n8 = RigVMDispatchNode(BoneName=root, Default=0, Name=deformRig_l_upperarm_bk_alpha) {template="DISPATCH_RigDispatch_GetAnimAttribute(in Name,in BoneName,in Default,out Value,out Found)"}
   > &RigUnit_SetCurveValue(Curve=l_arm_bk, Value=@n8.Value)
   n9 = RigVMVariableNode() {variable=Character Backward}
   n10 = &RigVMFunction_AlphaInterp(
    InterpSpeedDecreasing=10
    InterpSpeedIncreasing=20
    Scale=1
    bClampResult=False
    Value=&RigVMFunction_MathDoubleMax(A=@n8.Value, B=@n9.Value).Result
   )
   > &RigUnit_ModifyTransforms(
    ItemToModify=((Item=(Type=Bone,Name="stick_l"),Transform=(Rotation=(X=0,Y=0,Z=-0.2588,W=0.9659),Translation=(X=0,Y=0,Z=0),Scale3D=(X=1,Y=1,Z=1))))
    WeightMinimum=0
    Weight=@n10.Result
   )
   n11 = RigVMDispatchNode(BoneName=root, Default=0, Name=deformRig_r_upperarm_bk_alpha) {template="DISPATCH_RigDispatch_GetAnimAttribute(in Name,in BoneName,in Default,out Value,out Found)"}
   > &RigUnit_SetCurveValue(Curve=r_arm_bk, Value=@n11.Value)
   n12 = &RigVMFunction_AlphaInterp(
    InterpSpeedDecreasing=10
    InterpSpeedIncreasing=20
    Scale=1
    bClampResult=False
    Value=&RigVMFunction_MathDoubleMax(A=&RigVMFunction_MathDoubleMax(B=@n11.Value, A=@n9.Value).Result, B=RigVMVariableNode() {variable=Character Move Left}.Value).Result
   )
   > &RigUnit_ModifyTransforms(
    ItemToModify=((Item=(Type=Bone,Name="stick_r"),Transform=(Rotation=(X=0,Y=0,Z=-0.2588,W=0.9659),Translation=(X=0,Y=0,Z=0),Scale3D=(X=1,Y=1,Z=1))))
    WeightMinimum=0
    Weight=@n12.Result
   )
   n13 = &RigVMFunction_AlphaInterp(
    InterpSpeedDecreasing=10
    InterpSpeedIncreasing=20
    Scale=1
    bClampResult=False
    Value=RigVMVariableNode() {variable=Character Move Right}.Value
   )
   > &RigUnit_ModifyTransforms(
    ItemToModify=((Item=(Type=Bone,Name="stick_l"),Transform=(Rotation=(X=0,Y=-0.2588,Z=0,W=0.9659),Translation=(X=0,Y=0,Z=0),Scale3D=(X=1,Y=1,Z=1))))
    WeightMinimum=0
    Weight=@n13.Result
   )
   n14 = &RigVMFunction_AlphaInterp(
    InterpSpeedDecreasing=10
    InterpSpeedIncreasing=20
    Scale=1
    bClampResult=False
    Value=RigVMVariableNode() {variable=Character Move Left}.Value
   )
   > &RigUnit_ModifyTransforms(
    ItemToModify=((Item=(Type=Bone,Name="stick_l"),Transform=(Rotation=(X=0,Y=0.2588,Z=0,W=0.9659),Translation=(X=0,Y=0,Z=0),Scale3D=(X=1,Y=1,Z=1))))
    WeightMinimum=0
    Weight=@n14.Result
   )
   > RigVMFunctionReturnNode()
  ]}
 RigVMFunctionLibrary {nodes [
   DeltaRotation v4 = RigVMCollapseNode(
    ExecuteContext=()
    Rotation Order=ZYX
    child=(Type=Bone,Name="None")
    outputMinMax=(X=0,Y=1)
    parent=(Type=Bone,Name="None")
    premult=1
    rotationMinMax=(X=0,Y=0)
    twistAxis=(X=1,Y=0,Z=0)
   )
   arms = RigVMCollapseNode(ExecuteContext=())
   kyle_arms_and_levers = RigVMCollapseNode(ExecuteContext=())
   legs = RigVMCollapseNode(ExecuteContext=())
  ]}
 RigVMModel {nodes [
   RigUnit_BeginExecution = &RigUnit_BeginExecution()
   FN_DeformRig_FNC = RigVMFunctionReferenceNode(
    DeformRig_ Curve="(deformRig_l_upperarm_up_alpha_72_#1=true,deformRig_r_upperarm_up_alpha_73_#2=true,deformRig_l_upperarm_fr_alpha_74_#3=true,deformRig_r_upperarm_fr_alpha_75_CB51B80C4 ...(truncated)"
    DeformRig_Bool="(upperarm_1_#4=true,elbow_3_#5=false,wrist_7_#6=false,groin_8_#7=false,glute_17_#8=false,knee_9_40C408324C ...(truncated)"
    DeformRig_Debug=(upperarm_debug_18_#4=False,wrist_debug_20_#6=False,lat_debug_26_#9=False,groinGlute_debug_27_#7=False)
    DeformRig_Value="(upperarm_up_18_#10=(Rotation=(X=0.0,Y=-0.0,Z=0.0,W=1.0),Translation=(X=0.0,Y=0.0,Z=0),Scale3D=(X=1,Y=1.0,Z=1.0)),upperarm_fwd_fr_19_#11=(Rotation=(X=0.0,Y=-0.0,Z=0.0,W=1.0),Translation=(X=0.0,Y=0,Z ...(truncated)"
    ExecuteContext=@RigUnit_BeginExecution.ExecutePin
   )
   > legs = RigVMFunctionReferenceNode()
   > arms = RigVMFunctionReferenceNode()
   > kyle_arms_and_levers = RigVMFunctionReferenceNode()
  ]}
 RigVMModel Construction {nodes [
   n1 = &RigUnit_PrepareForExecution()
   &RigUnit_HierarchyAddNull(
    Name=hand_l_null
    Parent=(Type=Bone,Name="lowerarm_l")
    Space=GlobalSpace
    Transform=&RigUnit_GetTransform(Item=(Type=Bone,Name="hand_l"), bInitial=false).Transform
    ExecutePin=@n1.ExecutePin
   )
   > &RigUnit_HierarchyAddNull(
    Name=hand_r_null
    Parent=(Type=Bone,Name="lowerarm_r")
    Space=GlobalSpace
    Transform=&RigUnit_GetTransform(Item=(Type=Bone,Name="hand_r"), bInitial=false).Transform
   )
   > &RigUnit_HierarchyAddNull(
    Name=upperarm_l_null
    Parent=(Type=Bone,Name="clavicle_l")
    Space=GlobalSpace
    Transform=&RigUnit_GetTransform(Item=(Type=Bone,Name="upperarm_l"), bInitial=false).Transform
   )
   > &RigUnit_HierarchyAddNull(
    Name=upperarm_r_null
    Parent=(Type=Bone,Name="clavicle_r")
    Space=GlobalSpace
    Transform=&RigUnit_GetTransform(Item=(Type=Bone,Name="upperarm_r"), bInitial=false).Transform
   )
   > &RigUnit_HierarchyAddNull(
    Name=thigh_r_null
    Parent=(Type=Bone,Name="pelvis")
    Space=GlobalSpace
    Transform=&RigUnit_GetTransform(Item=(Type=Bone,Name="robot_thigh_r"), bInitial=false).Transform
   )
   > &RigUnit_HierarchyAddNull(
    Name=thigh_l_null
    Parent=(Type=Bone,Name="pelvis")
    Space=GlobalSpace
    Transform=&RigUnit_GetTransform(Item=(Type=Bone,Name="robot_thigh_l"), bInitial=false).Transform
   )
   > &RigUnit_HierarchyAddNull(
    Name=calf_l_null
    Parent=(Type=Bone,Name="thigh_l")
    Space=GlobalSpace
    Transform=&RigUnit_GetTransform(Item=(Type=Bone,Name="robot_calf_l"), bInitial=false).Transform
   )
   > &RigUnit_HierarchyAddNull(
    Name=calf_r_null
    Parent=(Type=Bone,Name="thigh_r")
    Space=GlobalSpace
    Transform=&RigUnit_GetTransform(Item=(Type=Bone,Name="robot_thigh_r"), bInitial=false).Transform
   )
  ]}
]
hierarchy [
 l_arm_ft {type=Curve, loc=<0 0 0>, rot=<0 0 0 1>}
 l_arm_bk {type=Curve, loc=<0 0 0>, rot=<0 0 0 1>}
 r_arm_ft {type=Curve, loc=<0 0 0>, rot=<0 0 0 1>}
 r_arm_bk {type=Curve, loc=<0 0 0>, rot=<0 0 0 1>}
 l_arm_up {type=Curve, loc=<0 0 0>, rot=<0 0 0 1>}
 r_arm_up {type=Curve, loc=<0 0 0>, rot=<0 0 0 1>}
 root {loc=<0 0 0>, rot=<0 0 0 1>}
 pelvis {parent=root, loc=<0 -0.7321 92.4218>, rot=<0.0127 0.707 -0.0127 -0.707>}
 spine_01 {parent=pelvis, loc=<5.6002 -0.5128 0>, rot=<0 0 0.034 -0.9994>}
 spine_02 {parent=spine_01, loc=<8.0756 0.0139 0>, rot=<0 0 0.0238 -0.9997>}
 spine_03 {parent=spine_02, loc=<7.0395 0 0>, rot=<0 0 0.0149 -0.9999>}
 spine_04 {parent=spine_03, loc=<8.1023 0 0>, rot=<0 0 -0.0132 -0.9999>}
 spine_05 {parent=spine_04, loc=<10.0141 0 0>, rot=<0 0 -0.0015 -1>}
 clavicle_l {parent=spine_05, loc=<2.5856 -0.2866 -4.3777>, rot=<0.0465 -0.7289 -0.009 -0.6829>}
 upperarm_l {parent=clavicle_l, loc=<18.2264 0.1342 -0.001>, rot=<0.0253 -0.2693 0.0013 -0.9627>}
 lowerarm_l {parent=upperarm_l, loc=<29.0071 0.1686 -1.239>, rot=<0.0391 -0.0099 -0.2453 -0.9686>}
 hand_l {parent=lowerarm_l, loc=<25.8607 0.222 1.0274>, rot=<-0.7193 -0.0988 0.1193 -0.6773>}
 thumb_01_l {parent=hand_l, loc=<1.3118 -0.3191 -2.459>, rot=<0.6546 -0.4042 0.0664 -0.6354>}
 thumb_02_l {parent=thumb_01_l, loc=<5.9336 -0.0448 0.0401>, rot=<-0.0037 0.0213 0.298 -0.9543>}
 thumb_03_l {parent=thumb_02_l, loc=<2.2223 0.3682 -0.0907>, rot=<0.0118 -0.013 0.0066 -0.9998>}
 index_metacarpal_l {parent=hand_l, loc=<1.6535 0.0183 -2.96>, rot=<0.0158 -0.0124 -0.0173 -0.9996>}
 index_01_l {parent=index_metacarpal_l, loc=<8.3008 0.0357 -0.0524>, rot=<-0.0044 0.0485 0.0924 -0.9945>}
 index_02_l {parent=index_01_l, loc=<3.3587 -0.3004 0.0452>, rot=<-0.002 0.0027 0.3225 -0.9466>}
 index_03_l {parent=index_02_l, loc=<2.5578 -0.123 -0.0252>, rot=<-0.002 0.0067 0.2558 -0.9667>}
 middle_metacarpal_l {parent=hand_l, loc=<2.1856 -0.1725 -0.5984>, rot=<0.006 0.0168 -0.0329 -0.9993>}
 middle_01_l {parent=middle_metacarpal_l, loc=<7.3275 -0.2447 0.0129>, rot=<-0.0208 0.0042 0.0858 -0.9961>}
 middle_02_l {parent=middle_01_l, loc=<4.0123 0.1401 0.1175>, rot=<-0.0218 0.0258 0.3512 -0.9357>}
 middle_03_l {parent=middle_02_l, loc=<2.6077 0.0534 0.1373>, rot=<0.0082 0.0188 0.1708 -0.9851>}
 ring_metacarpal_l {parent=hand_l, loc=<2.2074 -0.3625 1.3638>, rot=<-0.0284 0.0722 -0.0458 -0.9959>}
 ring_01_l {parent=ring_metacarpal_l, loc=<7.0295 -0.6197 -0.1446>, rot=<-0.0489 -0.0252 0.1892 -0.9804>}
 ring_02_l {parent=ring_01_l, loc=<3.6174 0.4647 -0.1027>, rot=<0 0 0.2806 -0.9598>}
 ring_03_l {parent=ring_02_l, loc=<2.4152 0.114 0.0199>, rot=<0 0 0.1189 -0.9929>}
 pinky_metacarpal_l {parent=hand_l, loc=<1.7616 -0.9753 3.3814>, rot=<-0.0397 0.0814 -0.069 -0.9935>}
 pinky_01_l {parent=pinky_metacarpal_l, loc=<7.1709 -0.4831 -0.1094>, rot=<-0.0654 -0.0035 0.2242 -0.9723>}
 pinky_02_l {parent=pinky_01_l, loc=<2.8274 0.0893 -0.1043>, rot=<0.0037 0.0011 0.2854 -0.9584>}
 pinky_03_l {parent=pinky_02_l, loc=<2.2046 0.0307 -0.0078>, rot=<-0.0002 0.0098 0.1445 -0.9895>}
 weapon_l {parent=hand_l, loc=<-0.1974 -3.4111 0.3811>, rot=<0 0 0 1>}
 hand_attach_l {parent=hand_l, loc=<8.5 -3 0>, rot=<0.7071 0.7071 0 0>}
 lowerarm_twist_02_l {parent=lowerarm_l, loc=<9.6199 0 0>, rot=<0 0 0 -1>}
 lowerarm_twist_01_l {parent=lowerarm_l, loc=<17.8549 0 0>, rot=<0 0 0 -1>}
 MSTR_skiblades {parent=lowerarm_twist_01_l, loc=<-1.8251 0.4326 7.6658>, rot=<-0.0869 -0.0713 0.0052 -0.9937>}
 skiblades {parent=MSTR_skiblades, loc=<0 0 0>, rot=<0 0 0 1>}
 dyn_skiblade_A {parent=skiblades, loc=<0.4112 5.9529 -1.4657>, rot=<0.2013 -0.0126 0.015 -0.9793>}
 dyn_skiblade_A_01 {parent=dyn_skiblade_A, loc=<2.7032 0.0495 -0.1427>, rot=<0.1397 0.0073 -0.0011 -0.9902>}
 dyn_skiblade_A_02 {parent=dyn_skiblade_A_01, loc=<12.1378 0 0>, rot=<0 -0.0612 0 -0.9981>}
 dyn_skiblade_B {parent=skiblades, loc=<0.0642 -7.0439 0.5441>, rot=<-0.1438 -0.0153 -0.0181 -0.9893>}
 dyn_skiblade_B_01 {parent=dyn_skiblade_B, loc=<3.6898 0.6343 0.5019>, rot=<0.0786 0.0085 0.0436 -0.9959>}
 dyn_skiblade_B_02 {parent=dyn_skiblade_B_01, loc=<11.916 0 0>, rot=<0 -0.1118 0 -0.9937>}
 skiblade_rope_A {parent=lowerarm_twist_01_l, loc=<-1.0807 7.0357 -0.0843>, rot=<0.0224 0.0306 0.0283 -0.9989>}
 upperarm_twist_01_l {parent=upperarm_l, loc=<7.9796 0.0843 -0.6195>, rot=<0 0 0 -1>}
 upperarm_twist_02_l {parent=upperarm_l, loc=<19.5059 0.0843 -0.6195>, rot=<0 0 0 -1>}
 robot_arm_l {parent=upperarm_l, loc=<0 0 0>, rot=<0 0 0 1>}
 robot_lowerarm_l {parent=robot_arm_l, loc=<29.0071 0.1686 -1.239>, rot=<0.0391 -0.0099 -0.2453 -0.9686>}
 MSTR_shoulderhelmet_l {parent=robot_arm_l, loc=<3.5431 0.05 11.0912>, rot=<-0.0119 0.0024 -0.0286 -0.9995>}
 shoulderhelmet_l {parent=MSTR_shoulderhelmet_l, loc=<0 0 0>, rot=<0 0 0 1>}
 shoulderhelmet_A_l {parent=robot_arm_l, loc=<4.3 10.9041 1.7736>, rot=<-0.0119 0.0024 -0.0286 -0.9995>}
 shoulderhelmet_B_l {parent=robot_arm_l, loc=<7.072 -10.2732 1.8654>, rot=<-0.0119 0.0024 -0.0286 -0.9995>}
 robot_arm_transition_l {parent=upperarm_l, loc=<0 0 0>, rot=<0 0 0 1>}
 neck_01 {parent=spine_05, loc=<10.815 1.9196 0>, rot=<0 0 -0.1864 -0.9825>}
 neck_02 {parent=neck_01, loc=<4.7347 0.3045 0>, rot=<0 0 0.0363 -0.9993>}
 head {parent=neck_02, loc=<5.7632 0.5552 0>, rot=<0 0 0.115 -0.9934>}
 MSTR_hat {parent=head, loc=<16.9594 -0.6775 -0.01>, rot=<-0.0284 -0.7065 0.0284 -0.7065>}
 MSTR_hat_r {parent=MSTR_hat, loc=<-13.595 -2.4482 -1.5497>, rot=<0.838 -0.0304 0.5445 -0.0197>}
 dyn_hat_r {parent=MSTR_hat_r, loc=<0 0 0>, rot=<0 0 0 1>}
 MSTR_hat_l {parent=MSTR_hat, loc=<13.575 -2.4482 -1.5501>, rot=<-0.0197 -0.5445 -0.0304 -0.838>}
 dyn_hat_l {parent=MSTR_hat_l, loc=<0 0 0>, rot=<0 0 0 1>}
 MSTR_kyle_arm_l {parent=neck_02, loc=<1.1323 3.2385 -11.2588>, rot=<-0.1062 -0.8933 0.0453 -0.4344>}
 kyle_shoulder_l {parent=MSTR_kyle_arm_l, loc=<0 0 0>, rot=<0 0 0 1>}
 kyle_elbow_l {parent=kyle_shoulder_l, loc=<3.9256 0 0>, rot=<-0.0086 -0.0716 -0.1047 -0.9919>}
 kyle_wrist_l {parent=kyle_elbow_l, loc=<2.7558 0 0>, rot=<0.0034 0.0195 0.0524 -0.9984>}
 MSTR_kyle_arm_r {parent=neck_02, loc=<1.1324 3.2385 11.2588>, rot=<0.8933 -0.1062 -0.4344 -0.0453>}
 kyle_shoulder_r {parent=MSTR_kyle_arm_r, loc=<0 0 0>, rot=<0 0 0 1>}
 kyle_elbow_r {parent=kyle_shoulder_r, loc=<-3.9253 0 -0.0003>, rot=<-0.0086 -0.0716 -0.1047 -0.9919>}
 kyle_wrist_r {parent=kyle_elbow_r, loc=<-2.756 0 0.0002>, rot=<0.0034 0.0195 0.0524 -0.9984>}
 seatbelt_A_l {parent=neck_02, loc=<1.988 6.5377 -10.1882>, rot=<-0.4498 -0.5841 -0.2207 -0.6386>}
 seatbelt_A_r {parent=neck_02, loc=<1.9883 6.5376 10.1882>, rot=<-0.5841 0.4498 0.6386 -0.2207>}
 seatbelt_B_l {parent=neck_02, loc=<5.3002 -1.0776 -11.0224>, rot=<-0.7012 -0.1375 0.2617 -0.6488>}
 seatbelt_B_r {parent=neck_02, loc=<5.2998 -1.0775 11.0224>, rot=<0.1375 -0.7012 -0.6488 -0.2617>}
 seatbelt_C_l {parent=neck_02, loc=<2.5629 -5.0974 -8.7978>, rot=<-0.4631 0.1147 0.4736 -0.7403>}
 seatbelt_C_r {parent=neck_02, loc=<2.5629 -5.0974 8.7978>, rot=<0.1147 0.4631 0.7403 0.4736>}
 clavicle_r {parent=spine_05, loc=<2.5855 -0.2866 4.3777>, rot=<0.7289 0.0465 -0.6829 0.009>}
 upperarm_r {parent=clavicle_r, loc=<-18.2264 -0.1342 0.001>, rot=<0.0253 -0.2693 0.0013 -0.9627>}
 lowerarm_r {parent=upperarm_r, loc=<-29.0071 -0.1686 1.239>, rot=<0.0391 -0.0099 -0.2453 -0.9686>}
 hand_r {parent=lowerarm_r, loc=<-25.8607 -0.222 -1.0274>, rot=<-0.7193 -0.0988 0.1193 -0.6773>}
 thumb_01_r {parent=hand_r, loc=<-1.3118 0.3191 2.459>, rot=<0.6546 -0.4042 0.0664 -0.6354>}
 thumb_02_r {parent=thumb_01_r, loc=<-5.9336 0.0448 -0.0401>, rot=<-0.0037 0.0213 0.298 -0.9543>}
 thumb_03_r {parent=thumb_02_r, loc=<-2.2223 -0.3682 0.0907>, rot=<0.0118 -0.013 0.0066 -0.9998>}
 index_metacarpal_r {parent=hand_r, loc=<-1.6535 -0.0183 2.96>, rot=<0.0158 -0.0124 -0.0173 -0.9996>}
 index_01_r {parent=index_metacarpal_r, loc=<-8.3008 -0.0357 0.0524>, rot=<-0.0044 0.0485 0.0924 -0.9945>}
 index_02_r {parent=index_01_r, loc=<-3.3587 0.3004 -0.0452>, rot=<-0.002 0.0027 0.3225 -0.9466>}
 index_03_r {parent=index_02_r, loc=<-2.5578 0.123 0.0252>, rot=<-0.002 0.0067 0.2558 -0.9667>}
 middle_metacarpal_r {parent=hand_r, loc=<-2.1856 0.1725 0.5984>, rot=<0.006 0.0168 -0.0329 -0.9993>}
 middle_01_r {parent=middle_metacarpal_r, loc=<-7.3275 0.2447 -0.0129>, rot=<-0.0208 0.0042 0.0858 -0.9961>}
 middle_02_r {parent=middle_01_r, loc=<-4.0123 -0.1401 -0.1175>, rot=<-0.0218 0.0258 0.3512 -0.9357>}
 middle_03_r {parent=middle_02_r, loc=<-2.6077 -0.0534 -0.1373>, rot=<0.0082 0.0188 0.1708 -0.9851>}
 ring_metacarpal_r {parent=hand_r, loc=<-2.2074 0.3625 -1.3638>, rot=<-0.0284 0.0722 -0.0458 -0.9959>}
 ring_01_r {parent=ring_metacarpal_r, loc=<-7.0295 0.6197 0.1446>, rot=<-0.0489 -0.0252 0.1892 -0.9804>}
 ring_02_r {parent=ring_01_r, loc=<-3.6174 -0.4647 0.1027>, rot=<0 0 0.2806 -0.9598>}
 ring_03_r {parent=ring_02_r, loc=<-2.4152 -0.114 -0.0199>, rot=<0 0 0.1189 -0.9929>}
 pinky_metacarpal_r {parent=hand_r, loc=<-1.7616 0.9753 -3.3814>, rot=<-0.0397 0.0814 -0.069 -0.9935>}
 pinky_01_r {parent=pinky_metacarpal_r, loc=<-7.1709 0.4831 0.1094>, rot=<-0.0654 -0.0035 0.2242 -0.9723>}
 pinky_02_r {parent=pinky_01_r, loc=<-2.8274 -0.0893 0.1043>, rot=<0.0037 0.0011 0.2854 -0.9584>}
 pinky_03_r {parent=pinky_02_r, loc=<-2.2046 -0.0307 0.0078>, rot=<-0.0002 0.0098 0.1445 -0.9895>}
 hand_attach_r {parent=hand_r, loc=<-8.5 3 0>, rot=<0 0 -0.7071 -0.7071>}
 weapon_r {parent=hand_r, loc=<0.1974 3.4111 -0.3811>, rot=<0 0 0 1>}
 lowerarm_twist_02_r {parent=lowerarm_r, loc=<-9.6199 0 0>, rot=<0 0 0 -1>}
 MSTR_shield_r {parent=lowerarm_twist_02_r, loc=<-3.3634 -0.2132 -5.5672>, rot=<-0.0654 -0.0193 -0.0318 -0.9972>}
 shield_r {parent=MSTR_shield_r, loc=<0 0 0>, rot=<0 0 0 1>}
 lowerarm_twist_01_r {parent=lowerarm_r, loc=<-17.8549 0 0>, rot=<0 0 0 -1>}
 upperarm_twist_01_r {parent=upperarm_r, loc=<-7.9796 -0.0843 0.6195>, rot=<0 0 0 -1>}
 upperarm_twist_02_r {parent=upperarm_r, loc=<-19.5059 -0.0843 0.6195>, rot=<0 0 0 -1>}
 robot_arm_r {parent=upperarm_r, loc=<0 0 0>, rot=<0 0 0 1>}
 robot_lowerarm_r {parent=robot_arm_r, loc=<-29.0071 -0.1686 1.239>, rot=<0.0391 -0.0099 -0.2453 -0.9686>}
 MSTR_shoulderhelmet_r {parent=robot_arm_r, loc=<-3.5431 -0.05 -11.0912>, rot=<-0.0119 0.0024 -0.0286 -0.9995>}
 shoulderhelmet_r {parent=MSTR_shoulderhelmet_r, loc=<0 0 0>, rot=<0 0 0 1>}
 shoulderhelmet_A_r {parent=robot_arm_r, loc=<-4.2997 -10.9041 -1.7738>, rot=<-0.0119 0.0024 -0.0286 -0.9995>}
 shoulderhelmet_B_r {parent=robot_arm_r, loc=<-7.0718 10.2733 -1.8657>, rot=<-0.0119 0.0024 -0.0286 -0.9995>}
 robot_arm_transition_r {parent=upperarm_r, loc=<0 0 0>, rot=<0 0 0 1>}
 attach_cape {parent=spine_05, loc=<9.017 -11.714 0>, rot=<0.0616 -0.7044 -0.0616 -0.7044>}
 cape {parent=attach_cape, type=Null, loc=<0 -2.9216 9.1139>, rot=<-0.0131 0 0 0.9999>}
 attach_backpack {parent=spine_05, loc=<-0.123 -13.473 0>, rot=<-0.0093 -0.707 0.0093 -0.707>}
 Backpack_BR {parent=attach_backpack, type=Null, loc=<0 -1.3959 5.2682>, rot=<-0.0871 0 0 0.9962>}
 stick_r {parent=spine_05, loc=<9.0117 19.1843 6.9818>, rot=<-0.0024 0.0066 0.2718 -0.9623>}
 stick_l {parent=spine_05, loc=<11.6276 17.5488 -6.4856>, rot=<0.0011 -0.0029 0.2712 -0.9625>}
 MSTR_wire_r {parent=spine_05, loc=<9.3754 27.7576 7.0446>, rot=<0.0284 -0.7065 -0.0284 -0.7065>}
 dyn_wire_r {parent=MSTR_wire_r, loc=<0 0 0>, rot=<0 0 0 1>}
 MSTR_wire_c {parent=spine_05, loc=<8.9715 25.7558 -0.6042>, rot=<0.0284 -0.7065 -0.0284 -0.7065>}
 dyn_wire_c {parent=MSTR_wire_c, loc=<0 0 0>, rot=<0 0 0 1>}
 MSTR_wire_l {parent=spine_05, loc=<9.7988 26.398 -11.5692>, rot=<0.0284 -0.7065 -0.0284 -0.7065>}
 dyn_wire_l {parent=MSTR_wire_l, loc=<0 0 0>, rot=<0 0 0 1>}
 seat {parent=spine_05, loc=<16.0432 -7.8747 -0.01>, rot=<0.0284 -0.7065 -0.0284 -0.7065>}
 thigh_l {parent=pelvis, loc=<-1.8716 -0.0109 -11.9936>, rot=<0.0012 0.0572 0.0155 -0.9982>}
 calf_l {parent=thigh_l, loc=<-41.9178 -0.2246 0.0926>, rot=<0.0006 -0.0161 -0.0391 -0.9991>}
 foot_l {parent=calf_l, loc=<-38.7057 1.2438 0.2452>, rot=<0.0975 -0.0474 0.0317 -0.9936>}
 ball_l {parent=foot_l, loc=<-6.1614 13.1469 0.6807>, rot=<0.0001 -0.0001 -0.7063 -0.7079>}
 calf_twist_02_l {parent=calf_l, loc=<-15.9268 -0.0967 0.0397>, rot=<0.0237 0.0095 -0.0035 -0.9997>}
 calf_twist_01_l {parent=calf_l, loc=<-32.9335 -0.0967 0.0397>, rot=<0.0237 0.0095 -0.0035 -0.9997>}
 robot_calf_l {parent=calf_l, loc=<0 0 0>, rot=<0 0 0 1>}
 robot_knee_transition_l {parent=calf_l, loc=<0 0 0>, rot=<0 0 0 1>}
 thigh_twist_01_l {parent=thigh_l, loc=<-21.6303 0 0>, rot=<0.0013 0 0 -1>}
 robot_thigh_l {parent=thigh_l, loc=<-11.0363 0.6012 1.3274>, rot=<0.0003 0.02 -0.0134 -0.9997>}
 robot_thigh_aim_l {parent=thigh_l, loc=<-41.9178 -0.2246 0.0926>, rot=<0.0006 -0.0161 -0.0391 -0.9991>}
 thigh_r {parent=pelvis, loc=<-1.8716 -0.0109 11.9936>, rot=<0.0572 -0.0012 0.9982 0.0155>}
 calf_r {parent=thigh_r, loc=<41.9178 0.2246 -0.0926>, rot=<0.0006 -0.0161 -0.0391 -0.9991>}
 foot_r {parent=calf_r, loc=<38.7057 -1.2438 -0.2452>, rot=<0.0975 -0.0474 0.0317 -0.9936>}
 ball_r {parent=foot_r, loc=<6.1614 -13.1469 -0.6807>, rot=<0.0001 -0.0001 -0.7063 -0.7079>}
 calf_twist_02_r {parent=calf_r, loc=<15.9268 0.0967 -0.0397>, rot=<0.0237 0.0095 -0.0035 -0.9997>}
 calf_twist_01_r {parent=calf_r, loc=<32.9335 0.0967 -0.0397>, rot=<0.0237 0.0095 -0.0035 -0.9997>}
 robot_calf_r {parent=calf_r, loc=<0 0 0>, rot=<0 0 0 1>}
 robot_knee_transition_r {parent=calf_r, loc=<0 0 0>, rot=<0 0 0 1>}
 thigh_twist_01_r {parent=thigh_r, loc=<21.6303 0 0>, rot=<0.0013 0 0 -1>}
 robot_thigh_r {parent=thigh_r, loc=<11.0364 -0.6012 -1.3274>, rot=<0.0003 0.02 -0.0134 -0.9997>}
 robot_thigh_aim_r {parent=thigh_r, loc=<41.9178 0.2246 -0.0926>, rot=<0.0006 -0.0161 -0.0391 -0.9991>}
 MSTR_handle {parent=pelvis, loc=<7.7666 1.1974 -0.01>, rot=<-0.463 -0.8864 0 0>}
 dyn_handle {parent=MSTR_handle, loc=<0 0 0>, rot=<0 0 0 1>}
 attach {parent=root, loc=<0 0 0>, rot=<0 0 0 1>}
 ik_hand_root {parent=root, loc=<0 0 0>, rot=<0 0 0 1>}
 ik_hand_gun {parent=ik_hand_root, loc=<-64.462 4.8638 101.1961>, rot=<0.6084 0.1793 0.4613 -0.6205>}
 ik_hand_r {parent=ik_hand_gun, loc=<0 0 0>, rot=<0 0 0 1>}
 ik_hand_l {parent=ik_hand_gun, loc=<65.7708 101.9258 43.6659>, rot=<-0.5102 -0.7906 -0.3387 0>}
 ik_foot_root {parent=root, loc=<0 0 0>, rot=<0 0 0 1>}
 ik_foot_l {parent=ik_foot_root, loc=<19.6189 -3.2051 10.3339>, rot=<0.0753 0.7018 0.0677 -0.7051>}
 ik_foot_r {parent=ik_foot_root, loc=<-19.6189 -3.2051 10.3339>, rot=<0.7051 0.0677 -0.7018 0.0753>}
 root_fp {parent=root, loc=<0 0 0>, rot=<0 0 0 1>}
 camera_root {parent=root_fp, loc=<0 0 0>, rot=<0 0 0 1>}
 camera {parent=camera_root, loc=<0 0 0>, rot=<0 0 0 1>}
 attach_fp {parent=root_fp, loc=<0 0 0>, rot=<0 0 0 1>}
 attach_emote_01 {parent=root, loc=<0 0 0>, rot=<0 0 0 1>}
 attach_emote_02 {parent=root, loc=<0 0 0>, rot=<0 0 0 1>}
 attach_emote_03 {parent=root, loc=<0 0 0>, rot=<0 0 0 1>}
 warp_point_01 {parent=root, loc=<0 0 0>, rot=<0 0 0 1>}
 warp_point_02 {parent=root, loc=<0 0 0>, rot=<0 0 0 1>}
 warp_point_03 {parent=root, loc=<0 0 0>, rot=<0 0 0 1>}
]
variables [
 Character Backward {type=double, default=0}
 Character Forward {type=double, default=0}
 Character Move Left {type=double, default=0}
 Character Move Right {type=double, default=0}
 FE idle {type=bool, default=False}
 FE idle alpha {type=double, default=0}
 is_emoting {type=bool, default=False}
 is_not_crouching {type=double, default=0}
 is_targeting {type=bool, default=False}
]
