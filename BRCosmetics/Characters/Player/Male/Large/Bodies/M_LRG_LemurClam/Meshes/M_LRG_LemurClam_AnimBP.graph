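# Animation Blueprint for the large male LemurClam body, adding procedural secondary motion to the pose copied from its attached parent mesh.
# A Control Rig takes movement and state inputs, then AnimDynamics springs handle, ski blade and wire bones, and DBNO additively rotates the ski blades.
# A front-end idle curve blends DefaultSlot onto the arms and robot arms.
# The update graph computes movement, stance and gravity values, though AnimDynamics uses constant gravity.
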AnimBlueprint /BRCosmetics/Characters/Player/Male/Large/Bodies/M_LRG_LemurClam/Meshes/M_LRG_LemurClam_AnimBP
target_skeleton=&/M_LRG_LemurClam_Skeleton
graphs [
 AnimGraph {nodes [
   n1 = Layered blend per bone(
    BasePose=Use cached pose 'base pose'()
    BlendPoses_0=Slot 'DefaultSlot'()
    BlendWeights_0=Get Custom Idle Is Playing Float() -> (Custom Idle Is Playing Float:double)
   ) {layerSetup [{branchFilters [{boneName=robot_arm_l}, {boneName=robot_arm_r}, {boneName=upperarm_l}, {boneName=upperarm_r}]}]}
   Output Pose(Result=@n1) {layerGroup=None}
   n2 = Control Rig!(
    Source=Copy Pose From Mesh() {bUseAttachedParent=true}
    Alpha
    Character Forward:double=Get Character Forward() -> (Character Forward:double)
    Character Backward:double=Get Character Backward() -> (Character Backward:double)
    Character Move Right:double=Get Character Move Right() -> (Character Move Right:double)
    Character Move Left:double=Get Character Move Left() -> (Character Move Left:double)
    is_not_crouching:double=Get is_not_crouching() -> (is_not_crouching:double)
    is_targeting:bool=Get bIsTargeting()
    is_emoting:bool=Get bIsPlayingEmote()
    FE idle:bool=Get Custom Idle Is Playing() -> (Custom Idle Is Playing:bool)
    FE idle alpha:double=Get Custom Idle Is Playing Float()
   ) -> (Pose) {controlRigAssetReference {blueprintRigClass=&/M_LRG_LemurClam_CtrlRig.M_LRG_LemurClam_CtrlRig_C}}
   n3 = Anim Dynamics - Bone: dyn_handle!(ComponentPose=Local To Component(LocalPose=@n2), Alpha=Get is_not_crouching(), GravityOverride:Vector="0,0,-980", ExternalForce="0,0,0") -> (Pose) {
    angularDampingOverride=0.999
    angularSpringConstant=7000
    bAngularSpring=true
    bGravityOverrideInSimSpace=true
    bOverrideAngularDamping=true
    bUseGravityOverride=true
    gravityScale=0.35
    simulationSpace=BoneRelative
    boundBone {boneName=dyn_handle}
    componentAppliedLinearAccClamp=<5000 5000 5000>
    componentLinearAccScale=<0.2 0.2 0.2>
    componentLinearVelScale=<0.25 0.25 0.25>
    gravityOverride {z=-980}
    physicsBodyDefinitions [{
      boundBone {boneName=dyn_handle}
      boxExtents=<20 30 30>
      constraintSetup {angularLimitsMax=<0 0 35>, angularLimitsMin=<0 0 -5>, angularTarget=<1 0 0>}
      localJointOffset=<1 -5 0>
     }]
    relativeSpaceBone {boneName=pelvis}
   }
   n4 = Anim Dynamics - Bone: dyn_skiblade_A!(ComponentPose=@n3, Alpha=1, GravityOverride:Vector="0,0,-980", ExternalForce="0,0,0") -> (Pose) {
    angularDampingOverride=0.9999
    angularSpringConstant=2000
    bAngularSpring=true
    bOverrideAngularDamping=true
    bUseGravityOverride=true
    bUsePlanarLimit=false
    gravityScale=0.35
    simulationSpace=RootRelative
    worldSpaceMinimumScale=0
    boundBone {boneName=dyn_skiblade_A}
    componentAppliedLinearAccClamp=<5000 5000 5000>
    componentLinearAccScale=<1 1 1>
    componentLinearVelScale=<1 1 1>
    gravityOverride {z=-980}
    physicsBodyDefinitions [{
      boundBone {boneName=dyn_skiblade_A}
      boxExtents=<30 10 30>
      constraintSetup {angularLimitsMax=<0 3 0>, angularLimitsMin=<0 -3 -5>, angularTarget=<1 0 0>}
      localJointOffset=<0.5 0 0>
     }]
    relativeSpaceBone {boneName=hand_l}
    simSpaceSettings {maxAngularAcceleration=5000, maxAngularVelocity=5000}
   }
   n5 = Anim Dynamics - Bone: dyn_skiblade_B!(ComponentPose=@n4, Alpha=1, GravityOverride:Vector="0,0,-980", ExternalForce="0,0,0") -> (Pose) {
    angularDampingOverride=0.9999
    angularSpringConstant=2000
    bAngularSpring=true
    bOverrideAngularDamping=true
    bUseGravityOverride=true
    bUsePlanarLimit=false
    gravityScale=0.35
    simulationSpace=RootRelative
    worldSpaceMinimumScale=0
    boundBone {boneName=dyn_skiblade_B}
    componentAppliedLinearAccClamp=<5000 5000 5000>
    componentLinearAccScale=<1 1 1>
    componentLinearVelScale=<1 1 1>
    gravityOverride {z=-980}
    physicsBodyDefinitions [{
      boundBone {boneName=dyn_skiblade_B}
      boxExtents=<30 10 30>
      constraintSetup {angularLimitsMax=<0 3 0>, angularLimitsMin=<0 -3 -5>, angularTarget=<1 0 0>}
      localJointOffset=<0.5 0 0>
     }]
    relativeSpaceBone {boneName=hand_l}
    simSpaceSettings {maxAngularAcceleration=5000, maxAngularVelocity=5000}
   }
   n6 = Anim Dynamics - Bone: dyn_wire_r!(ComponentPose=@n5, Alpha=1, GravityOverride:Vector="0,0,-980", ExternalForce="0,0,0") -> (Pose) {
    angularDampingOverride=0.999
    angularSpringConstant=30000
    bAngularSpring=true
    bOverrideAngularDamping=true
    bUseGravityOverride=true
    gravityScale=0.35
    simulationSpace=RootRelative
    boundBone {boneName=dyn_wire_r}
    componentLinearAccScale=<1 1 1>
    componentLinearVelScale=<1 1 1>
    gravityOverride {z=-980}
    physicsBodyDefinitions [{
      boundBone {boneName=dyn_wire_r}
      boxExtents=<20 30 30>
      constraintSetup {angularLimitsMax=<15 15 15>, angularLimitsMin=<-15 -15 -15>, angularTarget=<1 0 0>}
      localJointOffset=<4 0 0>
     }]
   }
   n7 = Anim Dynamics - Bone: dyn_wire_c!(ComponentPose=@n6, Alpha=1, GravityOverride:Vector="0,0,-980", ExternalForce="0,0,0") -> (Pose) {
    angularDampingOverride=0.999
    angularSpringConstant=30000
    bAngularSpring=true
    bOverrideAngularDamping=true
    bUseGravityOverride=true
    gravityScale=0.35
    simulationSpace=RootRelative
    boundBone {boneName=dyn_wire_c}
    componentLinearAccScale=<1 1 1>
    componentLinearVelScale=<1 1 1>
    gravityOverride {z=-980}
    physicsBodyDefinitions [{
      boundBone {boneName=dyn_wire_c}
      boxExtents=<20 30 30>
      constraintSetup {angularLimitsMax=<15 15 15>, angularLimitsMin=<-15 -15 -15>, angularTarget=<1 0 0>}
      localJointOffset=<4 0 0>
     }]
   }
   n8 = Anim Dynamics - Bone: dyn_wire_l!(ComponentPose=@n7, Alpha=1, GravityOverride:Vector="0,0,-980", ExternalForce="0,0,0") -> (Pose) {
    angularDampingOverride=0.999
    angularSpringConstant=30000
    bAngularSpring=true
    bOverrideAngularDamping=true
    bUseGravityOverride=true
    gravityScale=0.35
    simulationSpace=RootRelative
    boundBone {boneName=dyn_wire_l}
    componentLinearAccScale=<1 1 1>
    componentLinearVelScale=<1 1 1>
    gravityOverride {z=-980}
    physicsBodyDefinitions [{
      boundBone {boneName=dyn_wire_l}
      boxExtents=<20 30 30>
      constraintSetup {angularLimitsMax=<15 15 15>, angularLimitsMin=<-15 -15 -15>, angularTarget=<1 0 0>}
      localJointOffset=<4 0 0>
     }]
   }
   n9 = "Transform (Modify) Bone - Bone: MSTR_skiblades"(ComponentPose=@n8, Alpha=Get is_DBNO() -> (is_DBNO:double), Translation="0,0,0", Rotation="17.558,-5.441,-41.349", Scale="1,1,1") {rotationMode=BMM_Additive, boneToModify {boneName=MSTR_skiblades}, rotation=<rot 17.5576 -5.441 -41.3488>}
   Save cached pose 'base pose'(Pose=Component To Local(ComponentPose=@n9)) {cacheName=base pose}
  ]}
 EventGraph {nodes [
   Event BlueprintInitializeAnimation()
   > n1 = Cast To FortPlayerPawn(Object=TryGetFortPawn())
   > Set Pawn(Pawn:Fort Player Pawn Object Reference=@n1) -> (Output_Get:Fort Player Pawn Object Reference)
   Event BlueprintUpdateAnimation()
   > n2 = "Gravity Override State Function (Parameter Asset) (Nativized)"(
    CharacterPartAnimInstance=Self-Reference()
    GameTimeInSeconds=GetGameTimeInSeconds()
    DebugJointName=spine_05
    Parameters=&/M_LRG_LemurClam_Pack_Parameters
   )
   > Set Gravity Override(Gravity Override=@n2)
   > Set RBAN Settings Out(RBAN Settings Out=@n2.OutRigidBodyAnimNodeParameters)
   > Get!(execute:Exec) -> (then:Exec, else:Exec, Pawn:Fort Player Pawn Object Reference)
   > n3 = PlayerMovement(AnimInstance=Self-Reference(), XDotProductCoefficient=25, YDotProductCoefficient=25, ZDotProductCoefficient=1)
   > Set Character Move Right(Character Move Right:double=MapRangeClamped(Value=@n3, InRangeA=0, InRangeB=25)) -> (Output_Get:double)
   > Set Character Move Left(Character Move Left:double=MapRangeClamped(Value=@n3, InRangeA=0, InRangeB=-25)) -> (Output_Get:double)
   > Set Character Forward(Character Forward:double=MapRangeClamped(Value=@n3.YMovement, InRangeA=0, InRangeB=25)) -> (Output_Get:double)
   > Set Character Backward(Character Backward:double=MapRangeClamped(Value=@n3.YMovement, InRangeA=0, InRangeB=-25)) -> (Output_Get:double)
   n4 = Get bIsCrouching()
   > Set is_not_crouching(is_not_crouching:double="To Float (Boolean)"(InBool=NOT Boolean(A=@n4))) -> (Output_Get:double)
   > Set is_DBNO(is_DBNO:double="To Float (Boolean)"(InBool=Get bIsDBNO())) -> (Output_Get:double)
   > Branch(Condition=Get bIsInFrontEnd())
   > Set Custom Idle Is Playing Float(Custom Idle Is Playing Float:double=GetCurveValue(self=GetOwnerFrontEndAnimInstance(), CurveName=Frontend_Idle_Bypass)) -> (Output_Get:double)
   > n5 = Branch(Condition="float >= float"(A=Get Custom Idle Is Playing Float(), B=1))
   > Set Custom Idle Is Playing(Custom Idle Is Playing:bool=true) -> (Output_Get:bool)
   Set Custom Idle Is Playing(execute=@n5.else, Custom Idle Is Playing=false)
  ]}
]
variables [
 Character Backward {type=number}
 Character Forward {type=number}
 Character Move Left {type=number}
 Character Move Right {type=number}
 Custom Idle Is Playing {type=boolean}
 Custom Idle Is Playing Float {type=number}
 Gravity Override {type=Vector}
 is_DBNO {type=number}
 is_not_crouching {type=number}
 Pawn {type=FortPlayerPawn}
 RBAN Settings Out {
  type=FortRigidBodyAnimNodeParameters
  default {simSpaceSettings {dampingAlpha=0, maxAngularAcceleration=16, maxAngularVelocity=4, maxLinearAcceleration=800, maxLinearVelocity=200, velocityScaleZ=0.5}}
 }
]
