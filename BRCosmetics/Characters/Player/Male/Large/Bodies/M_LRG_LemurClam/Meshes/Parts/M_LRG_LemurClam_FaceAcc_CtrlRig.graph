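# Control Rig for LemurClam's face accessory, procedurally posing the small "kyle" arms, seatbelts and seat around neck and head.
# Setup builds wrist and elbow controls plus a head reference null.
# Smoothed upper-arm deform attributes and character movement variables offset wrist targets for stretchy two-bone IK, which are then reset.
# Seatbelts follow weighted parent constraints, while a swing-twist DeltaRotation function turns head rotation into weights for corrective seat and seatbelt offsets.
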ControlRigBlueprint /BRCosmetics/Characters/Player/Male/Large/Bodies/M_LRG_LemurClam/Meshes/Parts/M_LRG_LemurClam_FaceAcc_CtrlRig
graphs [
 New Function_ContainedGraph {nodes [
   n1 = RigVMFunctionEntryNode()
   RigVMFunction_MathBoolToInteger = &RigVMFunction_MathBoolToInteger(Value=@n1.onlyTwist)
   RigVMFunction_MathQuaternionSwingTwist = &RigVMFunction_MathQuaternionSwingTwist(
    twistAxis=@n1.twistAxis
    Input=&RigUnit_GetRelativeTransformForItem(bChildInitial=False, bParentInitial=False, child=@n1.child, parent=@n1.parent).RelativeTransform.Rotation
   )
   n2 = RigVMDispatchNode(
    Values=((),())
    Index=@RigVMFunction_MathBoolToInteger.Result
    Values.0=@RigVMFunction_MathQuaternionSwingTwist.Swing
    Values.1=@RigVMFunction_MathQuaternionSwingTwist.Twist
   ) {template="DISPATCH_RigVMDispatch_SelectInt32(in Index,in Values,out Result)"}
   RigVMFunction_MathQuaternionToEuler = &RigVMFunction_MathQuaternionToEuler(Value=@n2.Result, RotationOrder=@n1.Rotation Order)
   > DISPATCH_RigVMDispatch_Print = RigVMDispatchNode(ScreenColor=(R=1,G=1,B=1,A=1), ScreenDuration=1.5, Enabled=@n1.debugRotation, Value=@RigVMFunction_MathQuaternionToEuler.Result) {template="DISPATCH_RigVMDispatch_Print(in Prefix,in Value,in Enabled,in ScreenDuration,in ScreenColor)"}
   Cast = RigVMDispatchNode(Value=@n1.triggerAxis) {template="DISPATCH_RigVMDispatch_CastEnumToInt(in Value,out Result)"}
   DISPATCH_RigVMDispatch_SelectInt32 = RigVMDispatchNode(
    Values=(0,0,0)
    Index=@Cast.Result
    Values.0=@RigVMFunction_MathQuaternionToEuler.Result.X
    Values.1=@RigVMFunction_MathQuaternionToEuler.Result.Y
    Values.2=@RigVMFunction_MathQuaternionToEuler.Result.Z
   ) {template="DISPATCH_RigVMDispatch_SelectInt32(in Index,in Values,out Result)"}
   Remap = &RigVMFunction_MathDoubleRemap(
    bClamp=False
    Value=@DISPATCH_RigVMDispatch_SelectInt32.Result
    TargetMinimum=@n1.outputMinMax.X
    TargetMaximum=@n1.outputMinMax.Y
    SourceMinimum=@n1.rotationMinMax.X
    SourceMaximum=@n1.rotationMinMax.Y
   )
   Absolute = &RigVMFunction_MathDoubleAbs(Value=@Remap.Result)
   Multiply = &RigVMFunction_MathDoubleMul(
    A=&RigVMFunction_MathDoubleClamp(Value=RigVMDispatchNode(True=@Absolute.Result, Condition=@n1.absoluteValue, False=@Remap.Result) {template="DISPATCH_RigVMDispatch_If(in Condition,in True,in False,out Result)"}.Result).Result
    B=@n1.premult
   )
   > RigVMFunctionReturnNode(outputQuat=@n2.Result, outputWeight=@Multiply.Result)
  ]}
 RigVMFunctionLibrary {nodes [DeltaRotation v4 = RigVMCollapseNode(
    ExecuteContext=()
    Rotation Order=ZYX
    child=(Type=Bone,Name="None")
    outputMinMax=(X=0,Y=1)
    parent=(Type=Bone,Name="None")
    premult=1
    rotationMinMax=(X=0,Y=0)
    twistAxis=(X=1,Y=0,Z=0)
   )]}
 RigVMModel {nodes [
   n1 = &RigUnit_PrepareForExecution()
   Control Stack at Item_4 = RigVMFunctionReferenceNode(
    Add Bottom Null=true
    Add Null=true
    Add Secondary Control=False
    Control Suffix=_ctrl
    Item=(Name="kyle_wrist_l",Type=Bone)
    Name=kyle_wrist_l
    Null Suffix=_null
    Parent=(Name="spine_05",Type=Bone)
    Secondary Settings="(Shape=(bVisible=True,Name=\"Default\",Color=(R=1,G=0,B=0,A=1),Transform=(Rotation=(X=0,Y=0,Z=0,W=1),Translation=(X=0,Y=0,Z=0),Scale3D=(X=1,Y=1,Z=1))),Proxy=(bIsProxy=False,ShapeVisibility=BasedOnSelection),DisplayName=\"None\",InitialSpace=LocalSpace,bUsePref ...(truncated)"
    Settings="(Shape=(bVisible=True,Name=\"Arrow4_Thin\",Color=(R=0,G=1,B=0,A=1),Transform=(Rotation=(X=0,Y=-0.7071,Z=0,W=0.7071),Translation=(X=0,Y=0,Z=0),Scale3D=(X=1,Y=1,Z=1))),Proxy=(bIsProxy=False,ShapeVisibility=BasedOnSelection),DisplayName=\"None\",InitialSpace=Loca ...(truncated)"
    ExecuteContext=@n1.ExecutePin
   )
   > Control Stack at Item_5 = RigVMFunctionReferenceNode(
    Add Bottom Null=true
    Add Null=true
    Add Secondary Control=False
    Control Suffix=_ctrl
    Item=(Name="kyle_elbow_l",Type=Bone)
    Name=kyle_elbow_l
    Null Suffix=_null
    Parent=(Name="neck_02",Type=Bone)
    Secondary Settings="(Shape=(bVisible=True,Name=\"Default\",Color=(R=1,G=0,B=0,A=1),Transform=(Rotation=(X=0,Y=0,Z=0,W=1),Translation=(X=0,Y=0,Z=0),Scale3D=(X=1,Y=1,Z=1))),Proxy=(bIsProxy=False,ShapeVisibility=BasedOnSelection),DisplayName=\"None\",InitialSpace=LocalSpace,bUsePref ...(truncated)"
    Settings="(Shape=(bVisible=False,Name=\"Box_Thick\",Color=(R=0,G=1,B=0,A=1),Transform=(Rotation=(X=0,Y=0,Z=0,W=1),Translation=(X=0,Y=0,Z=0),Scale3D=(X=1,Y=1,Z=1))),Proxy=(bIsProxy=False,ShapeVisibility=BasedOnSelection),DisplayName=\"None\",InitialSpace=LocalSpace,bUseP ...(truncated)"
   )
   > Control Stack at Item_8 = RigVMFunctionReferenceNode(
    Add Bottom Null=true
    Add Null=true
    Add Secondary Control=False
    Control Suffix=_ctrl
    Item=(Name="kyle_wrist_r",Type=Bone)
    Name=kyle_wrist_r
    Null Suffix=_null
    Parent=(Name="spine_05",Type=Bone)
    Secondary Settings="(Shape=(bVisible=True,Name=\"Default\",Color=(R=1,G=0,B=0,A=1),Transform=(Rotation=(X=0,Y=0,Z=0,W=1),Translation=(X=0,Y=0,Z=0),Scale3D=(X=1,Y=1,Z=1))),Proxy=(bIsProxy=False,ShapeVisibility=BasedOnSelection),DisplayName=\"None\",InitialSpace=LocalSpace,bUsePref ...(truncated)"
    Settings="(Shape=(bVisible=True,Name=\"Arrow4_Thin\",Color=(R=0,G=1,B=0,A=1),Transform=(Rotation=(X=0,Y=-0.7071,Z=0,W=0.7071),Translation=(X=0,Y=0,Z=0),Scale3D=(X=1,Y=1,Z=1))),Proxy=(bIsProxy=False,ShapeVisibility=BasedOnSelection),DisplayName=\"None\",InitialSpace=Loca ...(truncated)"
   )
   > Control Stack at Item_9 = RigVMFunctionReferenceNode(
    Add Bottom Null=true
    Add Null=true
    Add Secondary Control=False
    Control Suffix=_ctrl
    Item=(Name="kyle_elbow_r",Type=Bone)
    Name=kyle_elbow_r
    Null Suffix=_null
    Parent=(Name="neck_02",Type=Bone)
    Secondary Settings="(Shape=(bVisible=True,Name=\"Default\",Color=(R=1,G=0,B=0,A=1),Transform=(Rotation=(X=0,Y=0,Z=0,W=1),Translation=(X=0,Y=0,Z=0),Scale3D=(X=1,Y=1,Z=1))),Proxy=(bIsProxy=False,ShapeVisibility=BasedOnSelection),DisplayName=\"None\",InitialSpace=LocalSpace,bUsePref ...(truncated)"
    Settings="(Shape=(bVisible=False,Name=\"Box_Thick\",Color=(R=0,G=1,B=0,A=1),Transform=(Rotation=(X=0,Y=0,Z=0,W=1),Translation=(X=0,Y=0,Z=0),Scale3D=(X=1,Y=1,Z=1))),Proxy=(bIsProxy=False,ShapeVisibility=BasedOnSelection),DisplayName=\"None\",InitialSpace=LocalSpace,bUseP ...(truncated)"
   )
   > &RigUnit_ModifyTransforms(
    ItemToModify=((Item=(Type=Null,Name="kyle_wrist_l_null"),Transform=(Rotation=(X=0.0283,Y=-0.2269,Z=0.7107,W=0.6653),Translation=(X=-7.9419,Y=8.9908,Z=-3.9259),Scale3D=(X=1,Y=1,Z=1))))
    Weight=1
    WeightMinimum=0
   )
   > &RigUnit_ModifyTransforms(
    ItemToModify=((Item=(Type=Null,Name="kyle_wrist_r_null"),Transform=(Rotation=(X=0.0441,Y=-0.2349,Z=0.6822,W=0.691),Translation=(X=5.4985,Y=-10.199,Z=5.3937),Scale3D=(X=1,Y=1,Z=1))))
    Weight=1
    WeightMinimum=0
   )
   > &RigUnit_HierarchyAddNull(
    Name=head_null
    Parent=(Type=Bone,Name="spine_05")
    Space=GlobalSpace
    Transform=&RigUnit_GetTransform(Item=(Type=Bone,Name="head"), bInitial=False).Transform
   )
   RigUnit_BeginExecution = &RigUnit_BeginExecution()
   FN_DeformRig_FNC = RigVMFunctionReferenceNode(
    DeformRig_ Curve="(deformRig_l_upperarm_up_alpha_72_#1=true,deformRig_r_upperarm_up_alpha_73_#2=true,deformRig_l_upperarm_fr_alpha_74_#3=true,deformRig_r_upperarm_fr_alpha_75_CB51B80C4 ...(truncated)"
    DeformRig_Bool="(upperarm_1_#4=true,elbow_3_#5=false,wrist_7_#6=false,groin_8_#7=false,glute_17_#8=false,knee_9_40C408324C ...(truncated)"
    DeformRig_Debug=(upperarm_debug_18_#4=False,wrist_debug_20_#6=False,lat_debug_26_#9=False,groinGlute_debug_27_#7=False)
    DeformRig_Value="(upperarm_up_18_#10=(Rotation=(X=0.0,Y=-0.0,Z=0.0,W=1.0),Translation=(X=0.0,Y=0.0,Z=0),Scale3D=(X=1,Y=1.0,Z=1.0)),upperarm_fwd_fr_19_#11=(Rotation=(X=0.0,Y=-0.0,Z=0.0,W=1.0),Translation=(X=0.0,Y=0,Z ...(truncated)"
    ExecuteContext=@RigUnit_BeginExecution.ExecutePin
   )
   > &RigUnit_ParentConstraint(
    AdvancedSettings=(InterpolationType=Average,RotationOrderForFilter=XZY)
    Child=(Type=Bone,Name="head_transition")
    Filter=(TranslationFilter=(bX=true,bY=true,bZ=true),RotationFilter=(bX=false,bY=True,bZ=True),ScaleFilter=(bX=True,bY=True,bZ=True))
    Parents=((Item=(Type=Bone,Name="head"),Weight=1))
    Weight=1
   )
   n2 = RigVMVariableNode() {variable=Character Forward}
   Maximum = &RigVMFunction_MathDoubleMax(
    B=RigVMDispatchNode(BoneName=root, Default=0, Name=deformRig_l_upperarm_fr_alpha) {template="DISPATCH_RigDispatch_GetAnimAttribute(in Name,in BoneName,in Default,out Value,out Found)"}.Value
    A=@n2.Value
   )
   n3 = &RigVMFunction_AlphaInterp(
    InterpSpeedDecreasing=10
    InterpSpeedIncreasing=20
    Scale=1
    bClampResult=False
    Value=&RigVMFunction_MathDoubleMax(
     B=RigVMDispatchNode(BoneName=root, Default=0, Name=deformRig_l_upperarm_up_alpha) {template="DISPATCH_RigDispatch_GetAnimAttribute(in Name,in BoneName,in Default,out Value,out Found)"}.Value
     A=@Maximum.Result
    ).Result
   )
   > &RigUnit_ModifyTransforms(
    ItemToModify=((Item=(Type=Null,Name="kyle_wrist_l_ctrl_null"),Transform=(Rotation=(X=0,Y=0,Z=0,W=1),Translation=(X=2.8199,Y=0.4253,Z=-0.0194),Scale3D=(X=1,Y=1,Z=1))))
    WeightMinimum=0
    Weight=@n3.Result
   )
   n4 = RigVMDispatchNode(BoneName=root, Default=0, Name=deformRig_r_upperarm_up_alpha) {template="DISPATCH_RigDispatch_GetAnimAttribute(in Name,in BoneName,in Default,out Value,out Found)"}
   n5 = &RigVMFunction_MathDoubleMax(
    B=RigVMDispatchNode(BoneName=root, Default=0, Name=deformRig_r_upperarm_fr_alpha) {template="DISPATCH_RigDispatch_GetAnimAttribute(in Name,in BoneName,in Default,out Value,out Found)"}.Value
    A=@n2.Value
   )
   n6 = &RigVMFunction_AlphaInterp(
    InterpSpeedDecreasing=10
    InterpSpeedIncreasing=20
    Scale=1
    bClampResult=False
    Value=&RigVMFunction_MathDoubleMax(A=&RigVMFunction_MathDoubleMax(B=@n4.Value, A=@n5.Result).Result, B=RigVMVariableNode() {variable=Character Move Right}.Value).Result
   )
   > &RigUnit_ModifyTransforms(
    ItemToModify=((Item=(Type=Null,Name="kyle_wrist_r_ctrl_null"),Transform=(Rotation=(X=0,Y=0,Z=0,W=1),Translation=(X=-2.859,Y=-0.8513,Z=0.6241),Scale3D=(X=1,Y=1,Z=1))))
    WeightMinimum=0
    Weight=@n6.Result
   )
   n7 = RigVMVariableNode() {variable=Character Backward}
   n8 = &RigVMFunction_AlphaInterp(
    InterpSpeedDecreasing=10
    InterpSpeedIncreasing=20
    Scale=1
    bClampResult=False
    Value=&RigVMFunction_MathDoubleMax(
     B=RigVMDispatchNode(BoneName=root, Default=0, Name=deformRig_l_upperarm_bk_alpha) {template="DISPATCH_RigDispatch_GetAnimAttribute(in Name,in BoneName,in Default,out Value,out Found)"}.Value
     A=@n7.Value
    ).Result
   )
   > &RigUnit_ModifyTransforms(
    ItemToModify=((Item=(Type=Null,Name="kyle_wrist_l_ctrl_null"),Transform=(Rotation=(X=0,Y=0,Z=0,W=1),Translation=(X=-2,Y=-1.5,Z=-1.5),Scale3D=(X=1,Y=1,Z=1))))
    WeightMinimum=0
    Weight=@n8.Result
   )
   n9 = RigVMDispatchNode(BoneName=root, Default=0, Name=deformRig_r_upperarm_bk_alpha) {template="DISPATCH_RigDispatch_GetAnimAttribute(in Name,in BoneName,in Default,out Value,out Found)"}
   n10 = &RigVMFunction_AlphaInterp(
    InterpSpeedDecreasing=10
    InterpSpeedIncreasing=20
    Scale=1
    bClampResult=False
    Value=&RigVMFunction_MathDoubleMax(A=&RigVMFunction_MathDoubleMax(B=@n9.Value, A=@n7.Value).Result, B=RigVMVariableNode() {variable=Character Move Left}.Value).Result
   )
   > &RigUnit_ModifyTransforms(
    ItemToModify=((Item=(Type=Null,Name="kyle_wrist_r_ctrl_null"),Transform=(Rotation=(X=0,Y=0,Z=0,W=1),Translation=(X=2.4714,Y=0.7517,Z=0.8138),Scale3D=(X=1,Y=1,Z=1))))
    WeightMinimum=0
    Weight=@n10.Result
   )
   n11 = &RigVMFunction_AlphaInterp(
    InterpSpeedDecreasing=10
    InterpSpeedIncreasing=20
    Scale=1
    bClampResult=False
    Value=RigVMVariableNode() {variable=Character Move Right}.Value
   )
   > &RigUnit_ModifyTransforms(
    ItemToModify=((Item=(Type=Null,Name="kyle_wrist_l_ctrl_null"),Transform=(Rotation=(X=-0.1461,Y=-0.006,Z=0.2558,W=0.9556),Translation=(X=1.151,Y=0.6664,Z=-1.7556),Scale3D=(X=1,Y=1,Z=1))))
    WeightMinimum=0
    Weight=@n11.Result
   )
   n12 = &RigVMFunction_AlphaInterp(
    InterpSpeedDecreasing=10
    InterpSpeedIncreasing=20
    Scale=1
    bClampResult=False
    Value=RigVMVariableNode() {variable=Character Move Left}.Value
   )
   > &RigUnit_ModifyTransforms(
    ItemToModify=((Item=(Type=Null,Name="kyle_wrist_l_ctrl_null"),Transform=(Rotation=(X=0,Y=0,Z=0,W=1),Translation=(X=-0.0173,Y=-3.0865,Z=0.8656),Scale3D=(X=1,Y=1,Z=1))))
    WeightMinimum=0
    Weight=@n12.Result
   )
   > /Script/ControlRig.RigUnit_TwoBoneIKSimplePerItem(
    DebugSettings=(bEnabled=False,Scale=10,WorldOffset=(Rotation=(X=0,Y=0,Z=0,W=1),Translation=(X=0,Y=0,Z=0),Scale3D=(X=1,Y=1,Z=1)))
    EffectorItem=(Type=Bone,Name="kyle_wrist_l")
    ItemA=(Type=Bone,Name="kyle_shoulder_l")
    ItemALength=0
    ItemB=(Type=Bone,Name="kyle_elbow_l")
    ItemBLength=0
    PoleVectorKind=Location
    PoleVectorSpace=(Type=Bone,Name="None")
    PrimaryAxis=(X=1,Y=0,Z=0)
    SecondaryAxis=(X=0,Y=0,Z=0)
    SecondaryAxisWeight=1
    StretchMaximumRatio=5
    StretchStartRatio=0.75
    Weight=1
    bEnableStretch=true
    bPropagateToChildren=True
    Effector=&RigUnit_GetTransform(Item=(Type=Null,Name="kyle_wrist_l_ctrl_null"), bInitial=False).Transform
    PoleVector=&RigUnit_GetTransform(Item=(Type=Null,Name="kyle_elbow_l_ctrl_null"), bInitial=False).Transform.Translation
   ) {is=struct RigVMUnitNode}
   > &RigUnit_TwoBoneIKSimplePerItem(
    DebugSettings=(bEnabled=False,Scale=10,WorldOffset=(Rotation=(X=0,Y=0,Z=0,W=1),Translation=(X=0,Y=0,Z=0),Scale3D=(X=1,Y=1,Z=1)))
    EffectorItem=(Type=Bone,Name="kyle_wrist_r")
    ItemA=(Type=Bone,Name="kyle_shoulder_r")
    ItemALength=0
    ItemB=(Type=Bone,Name="kyle_elbow_r")
    ItemBLength=0
    PoleVectorKind=Location
    PoleVectorSpace=(Type=Bone,Name="None")
    PrimaryAxis=(X=-1,Y=0,Z=0)
    SecondaryAxis=(X=0,Y=0,Z=0)
    SecondaryAxisWeight=1
    StretchMaximumRatio=5
    StretchStartRatio=0.75
    Weight=1
    bEnableStretch=true
    bPropagateToChildren=True
    Effector=&RigUnit_GetTransform(Item=(Type=Null,Name="kyle_wrist_r_ctrl_null"), bInitial=False).Transform
    PoleVector=&RigUnit_GetTransform(Item=(Type=Null,Name="kyle_elbow_r_ctrl_null"), bInitial=False).Transform.Translation
   )
   > &RigUnit_ParentConstraint(
    AdvancedSettings=(RotationOrderForFilter=XZY)
    Child=(Type=Bone,Name="seatbelt_B_l")
    Parents=((Item=(Type=Bone,Name="kyle_shoulder_l"),Weight=0.1341),(Item=(Type=Bone,Name="neck_02"),Weight=0.8659))
    Weight=1
   )
   > &RigUnit_ParentConstraint(
    AdvancedSettings=(RotationOrderForFilter=XZY)
    Child=(Type=Bone,Name="seatbelt_A_l")
    Parents=((Item=(Type=Bone,Name="spine_05"),Weight=0.0172),(Item=(Type=Bone,Name="neck_01"),Weight=0.0301),(Item=(Type=Bone,Name="neck_02"),Weight=0.9527))
    Weight=1
   )
   > &RigUnit_ParentConstraint(
    AdvancedSettings=(RotationOrderForFilter=XZY)
    Child=(Type=Bone,Name="seatbelt_C_l")
    Parents=((Item=(Type=Bone,Name="spine_05"),Weight=0.0153),(Item=(Type=Bone,Name="neck_01"),Weight=0.0361),(Item=(Type=Bone,Name="head_transition"),Weight=0.3579),(Item=(Type=Bone,Name="neck_02"),Weight=0.5907))
    Weight=1
   )
   > &RigUnit_ParentConstraint(
    AdvancedSettings=(InterpolationType=Shortest,RotationOrderForFilter=XZY)
    Child=(Type=Bone,Name="seatbelt_A_r")
    Parents=((Item=(Type=Bone,Name="kyle_shoulder_r"),Weight=0.2),(Item=(Type=Bone,Name="neck_02"),Weight=0.8))
    Weight=1
   )
   > &RigUnit_ParentConstraint(
    AdvancedSettings=(RotationOrderForFilter=XZY)
    Child=(Type=Bone,Name="seatbelt_B_r")
    Parents=((Item=(Type=Bone,Name="head_transition"),Weight=0.0121),(Item=(Type=Bone,Name="kyle_shoulder_r"),Weight=0.2),(Item=(Type=Bone,Name="neck_02"),Weight=0.7879))
    Weight=1
   )
   > &RigUnit_ParentConstraint(
    AdvancedSettings=(RotationOrderForFilter=XZY)
    Child=(Type=Bone,Name="seatbelt_C_r")
    Parents=((Item=(Type=Bone,Name="spine_05"),Weight=0.0153),(Item=(Type=Bone,Name="neck_01"),Weight=0.0361),(Item=(Type=Bone,Name="head_transition"),Weight=0.3579),(Item=(Type=Bone,Name="neck_02"),Weight=0.5907))
    Weight=1
   )
   > &RigUnit_ModifyTransforms(
    ItemToModify="((Item=(Type=Bone,Name=\"seatbelt_A_l\"),Transform=(Rotation=(X=-0.0154,Y=0.0962,Z=-0.1577,W=-0.9827),Translation=(X=0.2,Y=1.8,Z=0),Scale3D=(X=1,Y=1,Z=1))),(Item=(Type=Bone,Name=\"seatbelt_A_r\"),Transform=(Rotation=(X=0.1425,Y=-0.0208,Z=0.1431,W=0.9792),Trans ...(truncated)"
    Weight=1
    WeightMinimum=0
   )
   > &RigUnit_ModifyTransforms(
    ItemToModify="((Item=(Type=Null,Name=\"kyle_wrist_l_ctrl_null\"),Transform=(Rotation=(X=0,Y=0,Z=0,W=1),Translation=(X=0,Y=0,Z=0),Scale3D=(X=1,Y=1,Z=1))),(Item=(Type=Null,Name=\"kyle_wrist_r_ctrl_null\"),Transform=(Rotation=(X=0,Y=0,Z=0,W=1),Translation=(X=0,Y=0,Z=0),Scale3D ...(truncated)"
    Mode=OverrideLocal
    Weight=1
    WeightMinimum=0
   )
   > DeltaRotation v4_4 = RigVMFunctionReferenceNode(
    Rotation Order=ZYX
    child=(Type=Bone,Name="head")
    outputMinMax=(X=0,Y=1)
    parent=(Type=Null,Name="head_null")
    premult=1
    rotationMinMax=(X=-5,Y=-35)
    triggerAxis=Z
    twistAxis=(X=1,Y=0,Z=0)
   )
   > &RigUnit_ModifyTransforms(
    ItemToModify=((Item=(Type=Bone,Name="seat"),Transform=(Rotation=(X=0.2588,Y=0,Z=0,W=0.9659),Translation=(X=0,Y=0,Z=-4),Scale3D=(X=1,Y=1,Z=1))))
    WeightMinimum=0
    Weight=@"DeltaRotation v4_4".outputWeight
   )
   > DeltaRotation v4_1 = RigVMFunctionReferenceNode(
    Rotation Order=ZYX
    child=(Type=Bone,Name="head")
    onlyTwist=true
    outputMinMax=(X=0,Y=1)
    parent=(Type=Null,Name="head_null")
    premult=1
    rotationMinMax=(X=-10,Y=-40)
    triggerAxis=X
    twistAxis=(X=1,Y=0,Z=0)
   )
   > &RigUnit_ModifyTransforms(
    ItemToModify="((Item=(Type=Bone,Name=\"seatbelt_C_l\"),Transform=(Rotation=(X=0,Y=0,Z=0,W=1),Translation=(X=0,Y=0,Z=2),Scale3D=(X=1,Y=1,Z=1))),(Item=(Type=Bone,Name=\"seatbelt_B_l\"),Transform=(Rotation=(X=0,Y=0.0872,Z=0,W=0.9962),Translation=(X=1,Y=0,Z=0),Scale3D=(X=1,Y=1, ...(truncated)"
    WeightMinimum=0
    Weight=@"DeltaRotation v4_1".outputWeight
   )
   > DeltaRotation v4_2 = RigVMFunctionReferenceNode(
    Rotation Order=ZYX
    child=(Type=Bone,Name="head")
    onlyTwist=true
    outputMinMax=(X=0,Y=1)
    parent=(Type=Null,Name="head_null")
    premult=1
    rotationMinMax=(X=30,Y=70)
    triggerAxis=X
    twistAxis=(X=1,Y=0,Z=0)
   )
   > DeltaRotation v4_3 = RigVMFunctionReferenceNode(
    Rotation Order=ZYX
    child=(Type=Bone,Name="head")
    onlyTwist=false
    outputMinMax=(X=0,Y=1)
    parent=(Type=Null,Name="head_null")
    premult=1
    rotationMinMax=(X=0,Y=-20)
    triggerAxis=Y
    twistAxis=(X=1,Y=0,Z=0)
   )
   Multiply = &RigVMFunction_MathDoubleMul(A=@"DeltaRotation v4_2".outputWeight, B=@"DeltaRotation v4_3".outputWeight)
   > &RigUnit_ModifyTransforms(
    ItemToModify="((Item=(Type=Bone,Name=\"seatbelt_C_r\"),Transform=(Rotation=(X=0.0817,Y=0.2016,Z=-0.3667,W=0.9046),Translation=(X=0.4466,Y=1.5396,Z=1.5156),Scale3D=(X=1,Y=1,Z=1))),(Item=(Type=Bone,Name=\"seatbelt_B_r\"),Transform=(Rotation=(X=0,Y=0,Z=0,W=1),Translation=(X=0, ...(truncated)"
    WeightMinimum=0
    Weight=@Multiply.Result
   )
  ]}
]
hierarchy [
 l_arm_bk {type=Curve, loc=<0 0 0>, rot=<0 0 0 1>}
 l_arm_ft {type=Curve, loc=<0 0 0>, rot=<0 0 0 1>}
 l_arm_up {type=Curve, loc=<0 0 0>, rot=<0 0 0 1>}
 r_arm_bk {type=Curve, loc=<0 0 0>, rot=<0 0 0 1>}
 r_arm_ft {type=Curve, loc=<0 0 0>, rot=<0 0 0 1>}
 r_arm_up {type=Curve, loc=<0 0 0>, rot=<0 0 0 1>}
 root {loc=<0 0 0>, rot=<0 0 0 1>}
 pelvis {parent=root, loc=<0 -0.7321 92.4218>, rot=<0.0127 0.707 -0.0127 -0.707>}
 spine_01 {parent=pelvis, loc=<5.6002 -0.5128 0>, rot=<0 0 0.034 -0.9994>}
 spine_02 {parent=spine_01, loc=<8.0756 0.0139 0>, rot=<0 0 0.0238 -0.9997>}
 spine_03 {parent=spine_02, loc=<7.0395 0 0>, rot=<0 0 0.0149 -0.9999>}
 spine_04 {parent=spine_03, loc=<8.1023 0 0>, rot=<0 0 -0.0132 -0.9999>}
 spine_05 {parent=spine_04, loc=<10.0141 0 0>, rot=<0 0 -0.0015 -1>}
 neck_01 {parent=spine_05, loc=<10.815 1.9196 0>, rot=<0 0 -0.1864 -0.9825>}
 neck_02 {parent=neck_01, loc=<4.7347 0.3045 0>, rot=<0 0 0.0363 -0.9993>}
 head {parent=neck_02, loc=<5.7632 0.5552 0>, rot=<0 0 0.115 -0.9934>}
 MSTR_hat {parent=head, loc=<16.9594 -0.6775 -0.01>, rot=<-0.0284 -0.7065 0.0284 -0.7065>}
 MSTR_hat_r {parent=MSTR_hat, loc=<-13.595 -2.4482 -1.5497>, rot=<0.838 -0.0304 0.5445 -0.0197>}
 dyn_hat_r {parent=MSTR_hat_r, loc=<0 0 0>, rot=<0 0 0 1>}
 MSTR_hat_l {parent=MSTR_hat, loc=<13.575 -2.4482 -1.5501>, rot=<-0.0197 -0.5445 -0.0304 -0.838>}
 dyn_hat_l {parent=MSTR_hat_l, loc=<0 0 0>, rot=<0 0 0 1>}
 MSTR_kyle_arm_l {parent=neck_02, loc=<2.2503 -0.7615 -11.2588>, rot=<-0.1529 -0.8856 0.0285 -0.4376>}
 kyle_shoulder_l {parent=MSTR_kyle_arm_l, loc=<0 0 0>, rot=<0 0 0 1>}
 kyle_elbow_l {parent=kyle_shoulder_l, loc=<3.9256 0 0>, rot=<-0.0086 -0.0716 -0.1047 -0.9919>}
 kyle_wrist_l {parent=kyle_elbow_l, loc=<2.7558 0 0>, rot=<0.0034 0.0195 0.0524 -0.9984>}
 MSTR_kyle_arm_r {parent=neck_02, loc=<2.2505 -0.7615 11.2588>, rot=<0.8856 -0.1529 -0.4376 -0.0285>}
 kyle_shoulder_r {parent=MSTR_kyle_arm_r, loc=<0 0 0>, rot=<0 0 0 1>}
 kyle_elbow_r {parent=kyle_shoulder_r, loc=<-3.9259 0 0.0005>, rot=<-0.0086 -0.0716 -0.1047 -0.9919>}
 kyle_wrist_r {parent=kyle_elbow_r, loc=<-2.7556 -0.0001 -0.0002>, rot=<0.0034 0.0195 0.0524 -0.9984>}
 seatbelt_A_l {parent=neck_02, loc=<2.8123 2.8926 -10.1383>, rot=<-0.5327 -0.5097 -0.122 -0.6645>}
 seatbelt_A_r {parent=neck_02, loc=<2.8126 2.8925 10.2381>, rot=<-0.5097 0.5327 0.6645 -0.122>}
 seatbelt_B_l {parent=neck_02, loc=<5.3002 -1.0776 -11.0224>, rot=<-0.7012 -0.1375 0.2617 -0.6488>}
 seatbelt_B_r {parent=neck_02, loc=<5.2998 -1.0775 11.0224>, rot=<0.1375 -0.7012 -0.6488 -0.2617>}
 seatbelt_C_l {parent=neck_02, loc=<2.5629 -5.0974 -8.7978>, rot=<-0.4631 0.1147 0.4736 -0.7403>}
 seatbelt_C_r {parent=neck_02, loc=<2.5629 -5.0974 8.7978>, rot=<0.1147 0.4631 0.7403 0.4736>}
 head_transition {parent=neck_02, loc=<5.7632 0.5552 0>, rot=<0 0 0.115 -0.9934>}
 seat {parent=spine_05, loc=<16.0432 -7.8747 -0.01>, rot=<0.0284 -0.7065 -0.0284 -0.7065>}
 clavicle_l {parent=spine_05, loc=<2.5856 -0.2866 -4.3777>, rot=<0.0465 -0.7289 -0.009 -0.6829>}
 upperarm_l {parent=clavicle_l, loc=<18.2264 0.1342 -0.001>, rot=<0.0253 -0.2693 0.0013 -0.9627>}
 lowerarm_l {parent=upperarm_l, loc=<29.0071 0.1686 -1.239>, rot=<0.0391 -0.0099 -0.2453 -0.9686>}
 hand_l {parent=lowerarm_l, loc=<25.8607 0.222 1.0274>, rot=<-0.7193 -0.0988 0.1193 -0.6773>}
 weapon_l {parent=hand_l, loc=<-0.1974 -3.4111 0.3811>, rot=<0 0 0 1>}
 hand_attach_l {parent=hand_l, loc=<8.5 -3 0>, rot=<0.7071 0.7071 0 0>}
 thumb_01_l {parent=hand_l, loc=<1.3118 -0.3191 -2.459>, rot=<0.6546 -0.4042 0.0664 -0.6354>}
 thumb_02_l {parent=thumb_01_l, loc=<5.9336 -0.0448 0.0401>, rot=<-0.0037 0.0213 0.298 -0.9543>}
 thumb_03_l {parent=thumb_02_l, loc=<2.2223 0.3682 -0.0907>, rot=<0.0118 -0.013 0.0066 -0.9998>}
 index_metacarpal_l {parent=hand_l, loc=<1.6535 0.0183 -2.96>, rot=<0.0158 -0.0124 -0.0173 -0.9996>}
 index_01_l {parent=index_metacarpal_l, loc=<8.3008 0.0357 -0.0524>, rot=<-0.0044 0.0485 0.0924 -0.9945>}
 index_02_l {parent=index_01_l, loc=<3.3587 -0.3004 0.0452>, rot=<-0.002 0.0027 0.3225 -0.9466>}
 index_03_l {parent=index_02_l, loc=<2.5578 -0.123 -0.0252>, rot=<-0.002 0.0067 0.2558 -0.9667>}
 middle_metacarpal_l {parent=hand_l, loc=<2.1856 -0.1725 -0.5984>, rot=<0.006 0.0168 -0.0329 -0.9993>}
 middle_01_l {parent=middle_metacarpal_l, loc=<7.3275 -0.2447 0.0129>, rot=<-0.0208 0.0042 0.0858 -0.9961>}
 middle_02_l {parent=middle_01_l, loc=<4.0123 0.1401 0.1175>, rot=<-0.0218 0.0258 0.3512 -0.9357>}
 middle_03_l {parent=middle_02_l, loc=<2.6077 0.0534 0.1373>, rot=<0.0082 0.0188 0.1708 -0.9851>}
 ring_metacarpal_l {parent=hand_l, loc=<2.2074 -0.3625 1.3638>, rot=<-0.0284 0.0722 -0.0458 -0.9959>}
 ring_01_l {parent=ring_metacarpal_l, loc=<7.0295 -0.6197 -0.1446>, rot=<-0.0489 -0.0252 0.1892 -0.9804>}
 ring_02_l {parent=ring_01_l, loc=<3.6174 0.4647 -0.1027>, rot=<0 0 0.2806 -0.9598>}
 ring_03_l {parent=ring_02_l, loc=<2.4152 0.114 0.0199>, rot=<0 0 0.1189 -0.9929>}
 pinky_metacarpal_l {parent=hand_l, loc=<1.7616 -0.9753 3.3814>, rot=<-0.0397 0.0814 -0.069 -0.9935>}
 pinky_01_l {parent=pinky_metacarpal_l, loc=<7.1709 -0.4831 -0.1094>, rot=<-0.0654 -0.0035 0.2242 -0.9723>}
 pinky_02_l {parent=pinky_01_l, loc=<2.8274 0.0893 -0.1043>, rot=<0.0037 0.0011 0.2854 -0.9584>}
 pinky_03_l {parent=pinky_02_l, loc=<2.2046 0.0307 -0.0078>, rot=<-0.0002 0.0098 0.1445 -0.9895>}
 lowerarm_twist_02_l {parent=lowerarm_l, loc=<9.6199 0 0>, rot=<0 0 0 -1>}
 lowerarm_twist_01_l {parent=lowerarm_l, loc=<17.8549 0 0>, rot=<0 0 0 -1>}
 MSTR_skiblades {parent=lowerarm_twist_01_l, loc=<-1.8251 0.4326 7.6658>, rot=<-0.0869 -0.0713 0.0052 -0.9937>}
 skiblades {parent=MSTR_skiblades, loc=<0 0 0>, rot=<0 0 0 1>}
 dyn_skiblade_A {parent=skiblades, loc=<0.3213 6.8008 1.1135>, rot=<0.2013 -0.0126 0.015 -0.9793>}
 dyn_skiblade_B {parent=skiblades, loc=<0.0642 -7.0439 0.5441>, rot=<-0.1438 -0.0153 -0.0181 -0.9893>}
 upperarm_twist_01_l {parent=upperarm_l, loc=<7.9796 0.0843 -0.6195>, rot=<0 0 0 -1>}
 upperarm_twist_02_l {parent=upperarm_l, loc=<19.5059 0.0843 -0.6195>, rot=<0 0 0 -1>}
 robot_arm_l {parent=upperarm_l, loc=<0 0 0>, rot=<0 0 0 1>}
 robot_lowerarm_l {parent=robot_arm_l, loc=<29.0071 0.1686 -1.239>, rot=<0.0391 -0.0099 -0.2453 -0.9686>}
 MSTR_shoulderhelmet_l {parent=robot_arm_l, loc=<3.5431 0.05 11.0912>, rot=<-0.0119 0.0024 -0.0286 -0.9995>}
 shoulderhelmet_A_l {parent=robot_arm_l, loc=<4.3 10.9041 1.7736>, rot=<-0.0119 0.0024 -0.0286 -0.9995>}
 shoulderhelmet_B_l {parent=robot_arm_l, loc=<7.072 -10.2732 1.8654>, rot=<-0.0119 0.0024 -0.0286 -0.9995>}
 robot_arm_transition_l {parent=upperarm_l, loc=<0 0 0>, rot=<0 0 0 1>}
 clavicle_r {parent=spine_05, loc=<2.5855 -0.2866 4.3777>, rot=<0.7289 0.0465 -0.6829 0.009>}
 upperarm_r {parent=clavicle_r, loc=<-18.2264 -0.1342 0.001>, rot=<0.0253 -0.2693 0.0013 -0.9627>}
 lowerarm_r {parent=upperarm_r, loc=<-29.0071 -0.1686 1.239>, rot=<0.0391 -0.0099 -0.2453 -0.9686>}
 hand_r {parent=lowerarm_r, loc=<-25.8607 -0.222 -1.0274>, rot=<-0.7193 -0.0988 0.1193 -0.6773>}
 hand_attach_r {parent=hand_r, loc=<-8.5 3 0>, rot=<0 0 -0.7071 -0.7071>}
 thumb_01_r {parent=hand_r, loc=<-1.3118 0.3191 2.459>, rot=<0.6546 -0.4042 0.0664 -0.6354>}
 thumb_02_r {parent=thumb_01_r, loc=<-5.9336 0.0448 -0.0401>, rot=<-0.0037 0.0213 0.298 -0.9543>}
 thumb_03_r {parent=thumb_02_r, loc=<-2.2223 -0.3682 0.0907>, rot=<0.0118 -0.013 0.0066 -0.9998>}
 index_metacarpal_r {parent=hand_r, loc=<-1.6535 -0.0183 2.96>, rot=<0.0158 -0.0124 -0.0173 -0.9996>}
 index_01_r {parent=index_metacarpal_r, loc=<-8.3008 -0.0357 0.0524>, rot=<-0.0044 0.0485 0.0924 -0.9945>}
 index_02_r {parent=index_01_r, loc=<-3.3587 0.3004 -0.0452>, rot=<-0.002 0.0027 0.3225 -0.9466>}
 index_03_r {parent=index_02_r, loc=<-2.5578 0.123 0.0252>, rot=<-0.002 0.0067 0.2558 -0.9667>}
 middle_metacarpal_r {parent=hand_r, loc=<-2.1856 0.1725 0.5984>, rot=<0.006 0.0168 -0.0329 -0.9993>}
 middle_01_r {parent=middle_metacarpal_r, loc=<-7.3275 0.2447 -0.0129>, rot=<-0.0208 0.0042 0.0858 -0.9961>}
 middle_02_r {parent=middle_01_r, loc=<-4.0123 -0.1401 -0.1175>, rot=<-0.0218 0.0258 0.3512 -0.9357>}
 middle_03_r {parent=middle_02_r, loc=<-2.6077 -0.0534 -0.1373>, rot=<0.0082 0.0188 0.1708 -0.9851>}
 ring_metacarpal_r {parent=hand_r, loc=<-2.2074 0.3625 -1.3638>, rot=<-0.0284 0.0722 -0.0458 -0.9959>}
 ring_01_r {parent=ring_metacarpal_r, loc=<-7.0295 0.6197 0.1446>, rot=<-0.0489 -0.0252 0.1892 -0.9804>}
 ring_02_r {parent=ring_01_r, loc=<-3.6174 -0.4647 0.1027>, rot=<0 0 0.2806 -0.9598>}
 ring_03_r {parent=ring_02_r, loc=<-2.4152 -0.114 -0.0199>, rot=<0 0 0.1189 -0.9929>}
 pinky_metacarpal_r {parent=hand_r, loc=<-1.7616 0.9753 -3.3814>, rot=<-0.0397 0.0814 -0.069 -0.9935>}
 pinky_01_r {parent=pinky_metacarpal_r, loc=<-7.1709 0.4831 0.1094>, rot=<-0.0654 -0.0035 0.2242 -0.9723>}
 pinky_02_r {parent=pinky_01_r, loc=<-2.8274 -0.0893 0.1043>, rot=<0.0037 0.0011 0.2854 -0.9584>}
 pinky_03_r {parent=pinky_02_r, loc=<-2.2046 -0.0307 0.0078>, rot=<-0.0002 0.0098 0.1445 -0.9895>}
 weapon_r {parent=hand_r, loc=<0.1974 3.4111 -0.3811>, rot=<0 0 0 1>}
 lowerarm_twist_02_r {parent=lowerarm_r, loc=<-9.6199 0 0>, rot=<0 0 0 -1>}
 lowerarm_twist_01_r {parent=lowerarm_r, loc=<-17.8549 0 0>, rot=<0 0 0 -1>}
 upperarm_twist_01_r {parent=upperarm_r, loc=<-7.9796 -0.0843 0.6195>, rot=<0 0 0 -1>}
 upperarm_twist_02_r {parent=upperarm_r, loc=<-19.5059 -0.0843 0.6195>, rot=<0 0 0 -1>}
 attach_cape {parent=spine_05, loc=<9.017 -11.714 0>, rot=<0.0616 -0.7044 -0.0616 -0.7044>}
 attach_backpack {parent=spine_05, loc=<-0.123 -13.473 0>, rot=<-0.0093 -0.707 0.0093 -0.707>}
 stick_l {parent=spine_05, loc=<11.6276 17.5488 -6.4856>, rot=<0.0011 -0.0029 0.2712 -0.9625>}
 stick_r {parent=spine_05, loc=<9.0117 19.1843 6.9818>, rot=<-0.0024 0.0066 0.2718 -0.9623>}
 MSTR_wire_r {parent=spine_05, loc=<9.3754 27.7576 7.0446>, rot=<0.0284 -0.7065 -0.0284 -0.7065>}
 MSTR_wire_c {parent=spine_05, loc=<8.9715 25.7558 -0.6042>, rot=<0.0284 -0.7065 -0.0284 -0.7065>}
 MSTR_wire_l {parent=spine_05, loc=<9.7988 26.398 -11.5692>, rot=<0.0284 -0.7065 -0.0284 -0.7065>}
 thigh_l {parent=pelvis, loc=<-1.8716 -0.0109 -11.9936>, rot=<0.0012 0.0572 0.0155 -0.9982>}
 calf_l {parent=thigh_l, loc=<-41.9178 -0.2246 0.0926>, rot=<0.0006 -0.0161 -0.0391 -0.9991>}
 foot_l {parent=calf_l, loc=<-38.7057 1.2438 0.2452>, rot=<0.0975 -0.0474 0.0317 -0.9936>}
 ball_l {parent=foot_l, loc=<-6.1614 13.1469 0.6807>, rot=<0.0001 -0.0001 -0.7063 -0.7079>}
 calf_twist_02_l {parent=calf_l, loc=<-15.9268 -0.0967 0.0397>, rot=<0.0237 0.0095 -0.0035 -0.9997>}
 calf_twist_01_l {parent=calf_l, loc=<-32.9335 -0.0967 0.0397>, rot=<0.0237 0.0095 -0.0035 -0.9997>}
 robot_calf_l {parent=calf_l, loc=<0 0 0>, rot=<0 0 0 1>}
 robot_knee_transition_l {parent=calf_l, loc=<0 0 0>, rot=<0 0 0 1>}
 thigh_twist_01_l {parent=thigh_l, loc=<-21.6303 0 0>, rot=<0.0013 0 0 -1>}
 robot_thigh_l {parent=thigh_l, loc=<-11.0363 0.6012 1.3274>, rot=<0.0003 0.02 -0.0134 -0.9997>}
 robot_thigh_aim_l {parent=thigh_l, loc=<-41.9178 -0.2246 0.0926>, rot=<0.0006 -0.0161 -0.0391 -0.9991>}
 thigh_r {parent=pelvis, loc=<-1.8716 -0.0109 11.9936>, rot=<0.0572 -0.0012 0.9982 0.0155>}
 calf_r {parent=thigh_r, loc=<41.9178 0.2246 -0.0926>, rot=<0.0006 -0.0161 -0.0391 -0.9991>}
 foot_r {parent=calf_r, loc=<38.7057 -1.2438 -0.2452>, rot=<0.0975 -0.0474 0.0317 -0.9936>}
 ball_r {parent=foot_r, loc=<6.1614 -13.1469 -0.6807>, rot=<0.0001 -0.0001 -0.7063 -0.7079>}
 calf_twist_02_r {parent=calf_r, loc=<15.9268 0.0967 -0.0397>, rot=<0.0237 0.0095 -0.0035 -0.9997>}
 calf_twist_01_r {parent=calf_r, loc=<32.9335 0.0967 -0.0397>, rot=<0.0237 0.0095 -0.0035 -0.9997>}
 robot_calf_r {parent=calf_r, loc=<0 0 0>, rot=<0 0 0 1>}
 robot_knee_transition_r {parent=calf_r, loc=<0 0 0>, rot=<0 0 0 1>}
 thigh_twist_01_r {parent=thigh_r, loc=<21.6303 0 0>, rot=<0.0013 0 0 -1>}
 robot_thigh_r {parent=thigh_r, loc=<11.0364 -0.6012 -1.3274>, rot=<0.0003 0.02 -0.0134 -0.9997>}
 robot_thigh_aim_r {parent=thigh_r, loc=<41.9178 0.2246 -0.0926>, rot=<0.0006 -0.0161 -0.0391 -0.9991>}
]
variables [
 Character Backward {type=double, default=0}
 Character Forward {type=double, default=0}
 Character Move Left {type=double, default=0}
 Character Move Right {type=double, default=0}
]
