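# Physics asset for the face accessory of the large male LemurClam player body, covering the head and left and right hat bones.
# The head uses a sphere plus capsule, while each hat side has a capsule on its MSTR bone and another on its dyn bone.
# Each MSTR bone joins the head with free swing and twist.
# Each dyn bone joins its MSTR bone with asymmetric swing limits and a zero twist limit.
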PhysicsAsset /BRCosmetics/Characters/Player/Male/Large/Bodies/M_LRG_LemurClam/Meshes/Parts/M_LRG_LemurClam_FaceAcc_Physics
bodies {
 MSTR_hat_l [{shapeType=Capsule, length=6.6874, radius=1.8103, shapeName=MSTR_hat_l_capsule, rotation=<rot 0 0 90>}]
 MSTR_hat_r [{shapeType=Capsule, length=6.6874, radius=1.8103, shapeName=MSTR_hat_r_capsule, rotation=<rot 0 0 90>}]
 dyn_hat_l [{shapeType=Capsule, length=4, radius=4, shapeName=dyn_hat_l_capsule, center=<4 0 2>, rotation=<rot 90 0 0>}]
 dyn_hat_r [{shapeType=Capsule, length=4, radius=4, shapeName=dyn_hat_r_capsule, center=<-4 0 -2>, rotation=<rot 90 0 0>}]
 head [
  {shapeType=Sphere, radius=14.4196, shapeName=head_sphere, center=<2.9295 -1.0725 -0.0348>}
  {shapeType=Capsule, length=7.97, radius=12.03, shapeName=head_capsule, center=<19.579 -2.8406 -0.1733>, rotation=<rot 90 0 0>}
 ]
}
constraints [
 {bone1Name=dyn_hat_r, bone2Name=MSTR_hat_r, swing1LimitDegrees=10, swing2LimitDegrees=120, twistLimitDegrees=0}
 {
  bone1Name=MSTR_hat_r
  bone2Name=head
  swing1LimitDegrees=45
  swing1Motion=Free
  swing2LimitDegrees=45
  swing2Motion=Free
  twistLimitDegrees=45
  twistMotion=Free
 }
 {
  bone1Name=MSTR_hat_l
  bone2Name=head
  swing1LimitDegrees=45
  swing1Motion=Free
  swing2LimitDegrees=45
  swing2Motion=Free
  twistLimitDegrees=45
  twistMotion=Free
 }
 {bone1Name=dyn_hat_l, bone2Name=MSTR_hat_l, swing1LimitDegrees=10, swing2LimitDegrees=120, twistLimitDegrees=0}
]
